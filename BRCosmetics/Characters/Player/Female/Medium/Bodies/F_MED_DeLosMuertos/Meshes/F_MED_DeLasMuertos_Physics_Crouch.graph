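# Physics asset for the female medium DeLasMuertos character's skirt and lower body, named as the crouch variant.
# Capsules cover pelvis, thighs, calves and feet, linked by angularly free constraints.
# Skirt bone chains of capsules and boxes hang from the pelvis around front, diagonals, sides and back, mostly with limited swing and twist.
# Cross-constraints tie many neighbouring chains at matching segments, linking the strands into a connected sheet rather than independent chains.
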PhysicsAsset /BRCosmetics/Characters/Player/Female/Medium/Bodies/F_MED_DeLosMuertos/Meshes/F_MED_DeLasMuertos_Physics_Crouch
bodies {
 calf_l [{shapeType=Capsule, length=38.1885, radius=8.2327, center=<-24.7642 -2.0577 0.0001>, rotation=<rot 90 0 0>}]
 calf_r [{shapeType=Capsule, length=38.3683, radius=8.2608, center=<23.8321 2.6464 0.0173>, rotation=<rot 90 0 0>}]
 dyn_skirt_bk_01 [{shapeType=Capsule, length=17, radius=2, shapeName=dyn_skirt_frdiag_05_r_capsule, center=<5 0 0>, rotation=<rot 0 0 90>}]
 dyn_skirt_bk_01_l [{shapeType=Capsule, length=17, radius=2, shapeName=dyn_skirt_frdiag_05_r_capsule, center=<5 0 0>, rotation=<rot 0 0 90>}]
 dyn_skirt_bk_01_r [{shapeType=Capsule, length=16.1077, radius=2.8923, shapeName=dyn_skirt_frdiag_05_r_capsule, center=<5 0 0>, rotation=<rot 0 0 90>}]
 dyn_skirt_bk_02 [{shapeType=Capsule, length=17, radius=2.5, shapeName=dyn_skirt_frdiag_05_r_capsule, center=<5 0 0>, rotation=<rot 0 0 90>}]
 dyn_skirt_bk_02_l [{shapeType=Capsule, length=17, radius=2.5, shapeName=dyn_skirt_frdiag_05_r_capsule, center=<5 0 0>, rotation=<rot 0 0 90>}]
 dyn_skirt_bk_02_r [{shapeType=Capsule, length=16.1077, radius=3.3923, shapeName=dyn_skirt_frdiag_05_r_capsule, center=<5 0 0>, rotation=<rot 0 0 90>}]
 dyn_skirt_bk_03 [{shapeType=Capsule, length=14.4891, radius=5.5109, shapeName=dyn_skirt_frdiag_05_r_capsule, center=<8 0 0>, rotation=<rot 0 0 90>}]
 dyn_skirt_bk_03_l [{shapeType=Capsule, length=17, radius=3, shapeName=dyn_skirt_frdiag_05_r_capsule, center=<8 0 0>, rotation=<rot 0 0 90>}]
 dyn_skirt_bk_03_r [{shapeType=Capsule, length=14.4891, radius=5.5109, shapeName=dyn_skirt_frdiag_05_r_capsule, center=<8 0 0>, rotation=<rot 0 0 90>}]
 dyn_skirt_bk_04 [{shapeType=Capsule, length=19.5, radius=4.375, shapeName=dyn_skirt_frdiag_05_r_capsule, center=<7 0 0>, rotation=<rot 0 0 90>}]
 dyn_skirt_bk_04_l [{shapeType=Capsule, length=19.75, radius=4.5, shapeName=dyn_skirt_frdiag_05_r_capsule, center=<7 0 0>, rotation=<rot 0 0 90>}]
 dyn_skirt_bk_04_r [{shapeType=Capsule, length=19.75, radius=4.5, shapeName=dyn_skirt_frdiag_05_r_capsule, center=<7 0 0>, rotation=<rot 0 0 90>}]
 dyn_skirt_bk_05 [{shapeType=Capsule, length=15.253, radius=4.747, shapeName=dyn_skirt_frdiag_05_r_capsule, center=<4 0 0>, rotation=<rot 0 0 90>}]
 dyn_skirt_bk_05_l [{shapeType=Capsule, length=16.753, radius=4.997, shapeName=dyn_skirt_frdiag_05_r_capsule, center=<4 0 0>, rotation=<rot 0 0 90>}]
 dyn_skirt_bk_05_r [{shapeType=Capsule, length=16.753, radius=4.997, shapeName=dyn_skirt_frdiag_05_r_capsule, center=<4 0 0>, rotation=<rot 0 0 90>}]
 dyn_skirt_bk_A_01_l [{shapeType=Capsule, length=17, radius=2, shapeName=dyn_skirt_frdiag_05_r_capsule, center=<5 0 0>, rotation=<rot 0 0 90>}]
 dyn_skirt_bk_A_01_r [{shapeType=Capsule, length=17, radius=2, shapeName=dyn_skirt_frdiag_05_r_capsule, center=<5 0 0>, rotation=<rot 0 0 90>}]
 dyn_skirt_bk_A_02_l [{shapeType=Capsule, length=17, radius=2.5, shapeName=dyn_skirt_frdiag_05_r_capsule, center=<5 0 0>, rotation=<rot 0 0 90>}]
 dyn_skirt_bk_A_02_r [{shapeType=Capsule, length=17, radius=2.5, shapeName=dyn_skirt_frdiag_05_r_capsule, center=<5 0 0>, rotation=<rot 0 0 90>}]
 dyn_skirt_bk_A_03_l [{shapeType=Capsule, length=17, radius=3, shapeName=dyn_skirt_frdiag_05_r_capsule, center=<8 0 0>, rotation=<rot 0 0 90>}]
 dyn_skirt_bk_A_03_r [{shapeType=Capsule, length=14.4891, radius=5.5109, shapeName=dyn_skirt_frdiag_05_r_capsule, center=<8 0 0>, rotation=<rot 0 0 90>}]
 dyn_skirt_bk_A_04_l [{shapeType=Capsule, length=19.5, radius=4.375, shapeName=dyn_skirt_frdiag_05_r_capsule, center=<7 0 0>, rotation=<rot 0 0 90>}]
 dyn_skirt_bk_A_04_r [{shapeType=Capsule, length=19.5, radius=4.375, shapeName=dyn_skirt_frdiag_05_r_capsule, center=<7 0 0>, rotation=<rot 0 0 90>}]
 dyn_skirt_bk_A_05_l [{shapeType=Capsule, length=16.003, radius=4.997, shapeName=dyn_skirt_frdiag_05_r_capsule, center=<4 0 0>, rotation=<rot 0 0 90>}]
 dyn_skirt_bk_A_05_r [{shapeType=Capsule, length=16.003, radius=4.997, shapeName=dyn_skirt_frdiag_05_r_capsule, center=<4 0 0>, rotation=<rot 0 0 90>}]
 dyn_skirt_fr_01 [{shapeType=Box, extentX=10, extentY=5, extentZ=2, center=<10 0 0>}]
 dyn_skirt_fr_01_r [{shapeType=Box, extentX=10, extentY=1, extentZ=5, center=<6.1461 0 0>}]
 dyn_skirt_fr_02 [{shapeType=Box, extentX=10, extentY=5, extentZ=2, center=<10 0 0>}]
 dyn_skirt_fr_02_r [{shapeType=Box, extentX=10, extentY=1, extentZ=5, center=<5.6255 0 0>}]
 dyn_skirt_fr_03 [{shapeType=Box, extentX=10, extentY=5, extentZ=2, center=<10 0 0>}]
 dyn_skirt_fr_03_r [{shapeType=Box, extentX=10, extentY=1, extentZ=5, center=<2.5496 0 0>}]
 dyn_skirt_fr_04 [{shapeType=Box, extentX=10, extentY=5, extentZ=2, center=<10 0 0>}]
 dyn_skirt_fr_04_r [{shapeType=Box, extentX=10, extentY=1, extentZ=5, center=<1.844 0 0>}]
 dyn_skirt_fr_05_r [{shapeType=Box, extentX=10, extentY=1, extentZ=5, center=<1.4507 0 0>}]
 dyn_skirt_frdiag_01_l [{shapeType=Capsule, length=17, radius=2, shapeName=dyn_skirt_frdiag_05_r_capsule, center=<5 0 0>, rotation=<rot 0 0 90>}]
 dyn_skirt_frdiag_01_r [{shapeType=Box, extentX=15, extentY=10.3671, extentZ=3.53, center=<9.9708 0 0.6695>, rotation=<rot -10.6995 0 0>}]
 dyn_skirt_frdiag_02_l [{shapeType=Capsule, length=17.75, radius=3, shapeName=dyn_skirt_frdiag_05_r_capsule, center=<5 0 0>, rotation=<rot 0 0 90>}]
 dyn_skirt_frdiag_02_r [{shapeType=Box, extentX=15, extentY=11.3671, extentZ=3.53, center=<10 0 0.6629>}]
 dyn_skirt_frdiag_03_l [{shapeType=Capsule, length=17.875, radius=3.25, shapeName=dyn_skirt_frdiag_05_r_capsule, center=<5 0 0>, rotation=<rot 0 0 90>}]
 dyn_skirt_frdiag_03_r [{shapeType=Capsule, length=19.5, radius=3.5, shapeName=dyn_skirt_frdiag_05_r_capsule, center=<5 0 0>, rotation=<rot 0 0 90>}]
 dyn_skirt_frdiag_04_l [{shapeType=Capsule, length=17, radius=3, shapeName=dyn_skirt_frdiag_05_r_capsule, center=<5 0 0>, rotation=<rot 0 0 90>}]
 dyn_skirt_frdiag_04_r [{shapeType=Capsule, length=18, radius=3.125, shapeName=dyn_skirt_frdiag_05_r_capsule, center=<5 0 0>, rotation=<rot 0 0 90>}]
 dyn_skirt_frdiag_05_l [{shapeType=Capsule, length=17, radius=3, shapeName=dyn_skirt_frdiag_05_r_capsule, center=<5 0 0>, rotation=<rot 0 0 90>}]
 dyn_skirt_frdiag_05_r [{shapeType=Capsule, length=17, radius=3, shapeName=dyn_skirt_frdiag_05_r_capsule, center=<5 0 0>, rotation=<rot 0 0 90>}]
 dyn_skirt_frdiag_A_01_l [{shapeType=Box, extentX=10, extentY=5, extentZ=1, center=<10 0 0>}]
 dyn_skirt_frdiag_A_01_r [{shapeType=Box, extentX=10, extentY=5, extentZ=1, center=<10 0 0>}]
 dyn_skirt_frdiag_A_02_l [{shapeType=Box, extentX=15, extentY=10, extentZ=2, center=<10 0 0>}]
 dyn_skirt_frdiag_A_02_r [{shapeType=Box, extentX=15, extentY=10, extentZ=2, center=<10 0 0>}]
 dyn_skirt_frdiag_A_03_l [{shapeType=Box, extentX=15, extentY=10.5, extentZ=2, center=<10 0 0>}]
 dyn_skirt_frdiag_A_03_r [{shapeType=Box, extentX=15, extentY=16, extentZ=2, center=<10 0 0>}]
 dyn_skirt_frdiag_A_04_l [{shapeType=Box, extentX=15, extentY=10.5, extentZ=2, center=<10 0 0>}]
 dyn_skirt_frdiag_A_04_r [{shapeType=Box, extentX=15, extentY=10.5, extentZ=4, center=<10 0 0>}]
 dyn_skirt_frdiag_A_05_l [{shapeType=Box, extentX=7, extentY=11, extentZ=2, center=<10 0 0>}]
 dyn_skirt_frdiag_A_05_r [{shapeType=Box, extentX=7, extentY=11, extentZ=3.5, center=<10 0 0>}]
 dyn_skirt_out_01_l [{shapeType=Capsule, length=17, radius=2, shapeName=dyn_skirt_frdiag_05_r_capsule, center=<5 0 0>, rotation=<rot 0 0 90>}]
 dyn_skirt_out_01_r [{shapeType=Capsule, length=16.1077, radius=2.8923, shapeName=dyn_skirt_frdiag_05_r_capsule, center=<5 0 0>, rotation=<rot 0 0 90>}]
 dyn_skirt_out_02_l [{shapeType=Capsule, length=17.375, radius=2.75, shapeName=dyn_skirt_frdiag_05_r_capsule, center=<5 0 0>, rotation=<rot 0 0 90>}]
 dyn_skirt_out_02_r [{shapeType=Capsule, length=15.5043, radius=3.6423, shapeName=dyn_skirt_frdiag_05_r_capsule, center=<5 0 0>, rotation=<rot 0 0 90>}]
 dyn_skirt_out_03_l [{shapeType=Capsule, length=17, radius=3.125, shapeName=dyn_skirt_frdiag_05_r_capsule, center=<8 0 0>, rotation=<rot 0 0 90>}]
 dyn_skirt_out_03_r [{shapeType=Capsule, length=14.4891, radius=5.5109, shapeName=dyn_skirt_frdiag_05_r_capsule, center=<8 0 0>, rotation=<rot 0 0 90>}]
 dyn_skirt_out_04_l [{shapeType=Capsule, length=21, radius=4.25, shapeName=dyn_skirt_frdiag_05_r_capsule, center=<7 0 0>, rotation=<rot 0 0 90>}]
 dyn_skirt_out_04_r [{shapeType=Capsule, length=21, radius=4.25, shapeName=dyn_skirt_frdiag_05_r_capsule, center=<7 0 0>, rotation=<rot 0 0 90>}]
 dyn_skirt_out_05_l [{shapeType=Capsule, length=16.753, radius=4.997, shapeName=dyn_skirt_frdiag_05_r_capsule, center=<4 0 0>, rotation=<rot 0 0 90>}]
 dyn_skirt_out_05_r [{shapeType=Capsule, length=16.753, radius=4.997, shapeName=dyn_skirt_frdiag_05_r_capsule, center=<4 0 0>, rotation=<rot 0 0 90>}]
 foot_l [{shapeType=Capsule, length=14.747, radius=5.4757, center=<-2.6657 5.2706 -0.542>, rotation=<rot 0 0 -90>}]
 foot_r [{shapeType=Capsule, length=14.1611, radius=6.172, center=<2.6656 -6.3584 0.542>, rotation=<rot 0 0 -90>}]
 pelvis [
  {shapeType=Capsule, length=11.542, radius=6.7144, center=<-2.5931 -1.1459 0>}
  {shapeType=Capsule, length=2.75, radius=7.25, center=<-2.5987 -2.0584 0>}
 ]
 thigh_l [{shapeType=Capsule, length=23.4828, radius=9.3585, center=<-25.2038 -0.4536 1.1649>, rotation=<rot 90 0 0>}]
 thigh_r [{shapeType=Capsule, length=24.7917, radius=10.2672, center=<25.8306 0.6218 -0.8148>, rotation=<rot 90 0 0>}]
}
constraints [
 {
  bone1Name=calf_l
  bone2Name=thigh_l
  swing1LimitDegrees=45
  swing1Motion=Free
  swing2LimitDegrees=45
  swing2Motion=Free
  twistLimitDegrees=45
  twistMotion=Free
 }
 {
  bone1Name=foot_l
  bone2Name=calf_l
  swing1LimitDegrees=45
  swing1Motion=Free
  swing2LimitDegrees=45
  swing2Motion=Free
  twistLimitDegrees=45
  twistMotion=Free
 }
 {
  bone1Name=calf_r
  bone2Name=thigh_r
  swing1LimitDegrees=45
  swing1Motion=Free
  swing2LimitDegrees=45
  swing2Motion=Free
  twistLimitDegrees=45
  twistMotion=Free
 }
 {
  bone1Name=foot_r
  bone2Name=calf_r
  swing1LimitDegrees=45
  swing1Motion=Free
  swing2LimitDegrees=45
  swing2Motion=Free
  twistLimitDegrees=45
  twistMotion=Free
 }
 {
  bone1Name=thigh_l
  bone2Name=pelvis
  swing1LimitDegrees=45
  swing1Motion=Free
  swing2LimitDegrees=45
  swing2Motion=Free
  twistLimitDegrees=45
  twistMotion=Free
 }
 {
  bone1Name=thigh_r
  bone2Name=pelvis
  swing1LimitDegrees=45
  swing1Motion=Free
  swing2LimitDegrees=45
  swing2Motion=Free
  twistLimitDegrees=45
  twistMotion=Free
 }
 {bone1Name=dyn_skirt_frdiag_01_l, bone2Name=pelvis, swing1LimitDegrees=5, swing2LimitDegrees=10, twistLimitDegrees=2}
 {bone1Name=dyn_skirt_frdiag_02_l, bone2Name=dyn_skirt_frdiag_01_l, swing1LimitDegrees=2, swing2LimitDegrees=40, twistLimitDegrees=2}
 {bone1Name=dyn_skirt_frdiag_03_l, bone2Name=dyn_skirt_frdiag_02_l, swing1LimitDegrees=10, swing2LimitDegrees=40, twistLimitDegrees=5}
 {bone1Name=dyn_skirt_frdiag_04_l, bone2Name=dyn_skirt_frdiag_03_l, swing1LimitDegrees=20, swing2LimitDegrees=40, twistLimitDegrees=5}
 {bone1Name=dyn_skirt_frdiag_05_l, bone2Name=dyn_skirt_frdiag_04_l, swing1LimitDegrees=30, swing2LimitDegrees=40, twistLimitDegrees=5}
 {bone1Name=dyn_skirt_out_01_l, bone2Name=pelvis, swing1LimitDegrees=40, swing2LimitDegrees=40, twistLimitDegrees=5}
 {bone1Name=dyn_skirt_out_02_l, bone2Name=dyn_skirt_out_01_l, swing1LimitDegrees=40, swing2LimitDegrees=40, twistLimitDegrees=5}
 {bone1Name=dyn_skirt_out_03_l, bone2Name=dyn_skirt_out_02_l, swing1LimitDegrees=40, swing2LimitDegrees=40, twistLimitDegrees=5}
 {bone1Name=dyn_skirt_out_04_l, bone2Name=dyn_skirt_out_03_l, swing1LimitDegrees=40, swing2LimitDegrees=40, twistLimitDegrees=5}
 {bone1Name=dyn_skirt_out_05_l, bone2Name=dyn_skirt_out_04_l, swing1LimitDegrees=40, swing2LimitDegrees=40, twistLimitDegrees=5}
 {bone1Name=dyn_skirt_bk_01_l, bone2Name=pelvis, swing1LimitDegrees=40, swing2LimitDegrees=40, twistLimitDegrees=5}
 {bone1Name=dyn_skirt_bk_02_l, bone2Name=dyn_skirt_bk_01_l, swing1LimitDegrees=40, swing2LimitDegrees=40, twistLimitDegrees=5}
 {bone1Name=dyn_skirt_bk_03_l, bone2Name=dyn_skirt_bk_02_l, swing1LimitDegrees=40, swing2LimitDegrees=40, twistLimitDegrees=5}
 {bone1Name=dyn_skirt_bk_04_l, bone2Name=dyn_skirt_bk_03_l, swing1LimitDegrees=40, swing2LimitDegrees=40, twistLimitDegrees=5}
 {bone1Name=dyn_skirt_bk_05_l, bone2Name=dyn_skirt_bk_04_l, swing1LimitDegrees=40, swing2LimitDegrees=40, twistLimitDegrees=5}
 {bone1Name=dyn_skirt_fr_01, bone2Name=pelvis, swing1LimitDegrees=40, swing2LimitDegrees=40, twistLimitDegrees=5}
 {bone1Name=dyn_skirt_fr_02, bone2Name=dyn_skirt_fr_01, swing1LimitDegrees=40, swing2LimitDegrees=40, twistLimitDegrees=5}
 {bone1Name=dyn_skirt_fr_03, bone2Name=dyn_skirt_fr_02, swing1LimitDegrees=40, swing2LimitDegrees=40, twistLimitDegrees=5}
 {bone1Name=dyn_skirt_fr_04, bone2Name=dyn_skirt_fr_03, swing1LimitDegrees=40, swing2LimitDegrees=40, twistLimitDegrees=5}
 {bone1Name=dyn_skirt_bk_01, bone2Name=pelvis, swing1LimitDegrees=40, swing2LimitDegrees=40, twistLimitDegrees=5}
 {bone1Name=dyn_skirt_bk_02, bone2Name=dyn_skirt_bk_01, swing1LimitDegrees=40, swing2LimitDegrees=40, twistLimitDegrees=5}
 {bone1Name=dyn_skirt_bk_03, bone2Name=dyn_skirt_bk_02, swing1LimitDegrees=40, swing2LimitDegrees=40, twistLimitDegrees=5}
 {bone1Name=dyn_skirt_bk_04, bone2Name=dyn_skirt_bk_03, swing1LimitDegrees=40, swing2LimitDegrees=40, twistLimitDegrees=5}
 {bone1Name=dyn_skirt_bk_05, bone2Name=dyn_skirt_bk_04, swing1LimitDegrees=40, swing2LimitDegrees=40, twistLimitDegrees=5}
 {bone1Name=dyn_skirt_frdiag_01_r, bone2Name=pelvis, swing1LimitDegrees=5, swing2LimitDegrees=40, twistLimitDegrees=2}
 {bone1Name=dyn_skirt_frdiag_02_r, bone2Name=dyn_skirt_frdiag_01_r, swing1LimitDegrees=2, swing2LimitDegrees=40, twistLimitDegrees=2}
 {bone1Name=dyn_skirt_frdiag_03_r, bone2Name=dyn_skirt_frdiag_02_r, swing1LimitDegrees=10, swing2LimitDegrees=40, twistLimitDegrees=5}
 {bone1Name=dyn_skirt_frdiag_04_r, bone2Name=dyn_skirt_frdiag_03_r, swing1LimitDegrees=20, swing2LimitDegrees=40, twistLimitDegrees=5}
 {bone1Name=dyn_skirt_frdiag_05_r, bone2Name=dyn_skirt_frdiag_04_r, swing1LimitDegrees=30, swing2LimitDegrees=40, twistLimitDegrees=5}
 {bone1Name=dyn_skirt_out_01_r, bone2Name=pelvis, swing1LimitDegrees=40, swing2LimitDegrees=40, twistLimitDegrees=5}
 {bone1Name=dyn_skirt_out_02_r, bone2Name=dyn_skirt_out_01_r, swing1LimitDegrees=40, swing2LimitDegrees=40, twistLimitDegrees=5}
 {bone1Name=dyn_skirt_out_03_r, bone2Name=dyn_skirt_out_02_r, swing1LimitDegrees=40, swing2LimitDegrees=40, twistLimitDegrees=5}
 {bone1Name=dyn_skirt_out_04_r, bone2Name=dyn_skirt_out_03_r, swing1LimitDegrees=40, swing2LimitDegrees=40, twistLimitDegrees=5}
 {bone1Name=dyn_skirt_out_05_r, bone2Name=dyn_skirt_out_04_r, swing1LimitDegrees=40, swing2LimitDegrees=40, twistLimitDegrees=5}
 {bone1Name=dyn_skirt_bk_01_r, bone2Name=pelvis, swing1LimitDegrees=40, swing2LimitDegrees=40, twistLimitDegrees=5}
 {bone1Name=dyn_skirt_bk_02_r, bone2Name=dyn_skirt_bk_01_r, swing1LimitDegrees=40, swing2LimitDegrees=40, twistLimitDegrees=5}
 {bone1Name=dyn_skirt_bk_03_r, bone2Name=dyn_skirt_bk_02_r, swing1LimitDegrees=40, swing2LimitDegrees=40, twistLimitDegrees=5}
 {bone1Name=dyn_skirt_bk_04_r, bone2Name=dyn_skirt_bk_03_r, swing1LimitDegrees=40, swing2LimitDegrees=40, twistLimitDegrees=5}
 {bone1Name=dyn_skirt_bk_05_r, bone2Name=dyn_skirt_bk_04_r, swing1LimitDegrees=40, swing2LimitDegrees=40, twistLimitDegrees=5}
 {bone1Name=dyn_skirt_frdiag_A_01_r, bone2Name=pelvis, swing1LimitDegrees=5, swing2LimitDegrees=40, twistLimitDegrees=1}
 {bone1Name=dyn_skirt_frdiag_A_02_r, bone2Name=dyn_skirt_frdiag_A_01_r, swing1LimitDegrees=5, swing2LimitDegrees=40, twistLimitDegrees=1}
 {bone1Name=dyn_skirt_frdiag_A_03_r, bone2Name=dyn_skirt_frdiag_A_02_r, swing1LimitDegrees=10, swing2LimitDegrees=40, twistLimitDegrees=5}
 {bone1Name=dyn_skirt_frdiag_A_04_r, bone2Name=dyn_skirt_frdiag_A_03_r, swing1LimitDegrees=20, swing2LimitDegrees=40, twistLimitDegrees=5}
 {bone1Name=dyn_skirt_frdiag_A_05_r, bone2Name=dyn_skirt_frdiag_A_04_r, swing1LimitDegrees=30, swing2LimitDegrees=40, twistLimitDegrees=5}
 {bone1Name=dyn_skirt_frdiag_A_01_l, bone2Name=pelvis, swing1LimitDegrees=40, swing2LimitDegrees=40, twistLimitDegrees=40}
 {bone1Name=dyn_skirt_frdiag_A_02_l, bone2Name=dyn_skirt_frdiag_A_01_l, swing1LimitDegrees=5, swing2LimitDegrees=40, twistLimitDegrees=1}
 {bone1Name=dyn_skirt_frdiag_A_03_l, bone2Name=dyn_skirt_frdiag_A_02_l, swing1LimitDegrees=5, swing2LimitDegrees=40, twistLimitDegrees=5}
 {bone1Name=dyn_skirt_frdiag_A_04_l, bone2Name=dyn_skirt_frdiag_A_03_l, swing1LimitDegrees=30, swing2LimitDegrees=40, twistLimitDegrees=5}
 {bone1Name=dyn_skirt_frdiag_A_05_l, bone2Name=dyn_skirt_frdiag_A_04_l, swing1LimitDegrees=30.7143, swing2LimitDegrees=40, twistLimitDegrees=5}
 {bone1Name=dyn_skirt_bk_A_01_r, bone2Name=pelvis, swing1LimitDegrees=40, swing2LimitDegrees=40, twistLimitDegrees=5}
 {bone1Name=dyn_skirt_bk_A_02_r, bone2Name=dyn_skirt_bk_A_01_r, swing1LimitDegrees=40, swing2LimitDegrees=40, twistLimitDegrees=5}
 {bone1Name=dyn_skirt_bk_A_03_r, bone2Name=dyn_skirt_bk_A_02_r, swing1LimitDegrees=40, swing2LimitDegrees=40, twistLimitDegrees=5}
 {bone1Name=dyn_skirt_bk_A_04_r, bone2Name=dyn_skirt_bk_A_03_r, swing1LimitDegrees=40, swing2LimitDegrees=40, twistLimitDegrees=5}
 {bone1Name=dyn_skirt_bk_A_05_r, bone2Name=dyn_skirt_bk_A_04_r, swing1LimitDegrees=40, swing2LimitDegrees=40, twistLimitDegrees=5}
 {bone1Name=dyn_skirt_bk_A_01_l, bone2Name=pelvis, swing1LimitDegrees=40, swing2LimitDegrees=40, twistLimitDegrees=5}
 {bone1Name=dyn_skirt_bk_A_02_l, bone2Name=dyn_skirt_bk_A_01_l, swing1LimitDegrees=40, swing2LimitDegrees=40, twistLimitDegrees=5}
 {bone1Name=dyn_skirt_bk_A_03_l, bone2Name=dyn_skirt_bk_A_02_l, swing1LimitDegrees=40, swing2LimitDegrees=40, twistLimitDegrees=5}
 {bone1Name=dyn_skirt_bk_A_04_l, bone2Name=dyn_skirt_bk_A_03_l, swing1LimitDegrees=40, swing2LimitDegrees=40, twistLimitDegrees=5}
 {bone1Name=dyn_skirt_bk_A_05_l, bone2Name=dyn_skirt_bk_A_04_l, swing1LimitDegrees=40, swing2LimitDegrees=40, twistLimitDegrees=5}
 {bone1Name=dyn_skirt_bk_A_03_l, bone2Name=dyn_skirt_bk_03, swing1LimitDegrees=40, swing2LimitDegrees=40, twistLimitDegrees=5}
 {bone1Name=dyn_skirt_bk_A_03_r, bone2Name=dyn_skirt_bk_03, swing1LimitDegrees=40, swing2LimitDegrees=40, twistLimitDegrees=5}
 {bone1Name=dyn_skirt_bk_A_04_l, bone2Name=dyn_skirt_bk_04, swing1LimitDegrees=40, swing2LimitDegrees=40, twistLimitDegrees=5}
 {bone1Name=dyn_skirt_bk_A_04_r, bone2Name=dyn_skirt_bk_04, swing1LimitDegrees=40, swing2LimitDegrees=40, twistLimitDegrees=5}
 {bone1Name=dyn_skirt_bk_A_05_l, bone2Name=dyn_skirt_bk_05, swing1LimitDegrees=40, swing2LimitDegrees=40, twistLimitDegrees=5}
 {bone1Name=dyn_skirt_bk_A_05_r, bone2Name=dyn_skirt_bk_05, swing1LimitDegrees=40, swing2LimitDegrees=40, twistLimitDegrees=5}
 {bone1Name=dyn_skirt_bk_A_05_r, bone2Name=dyn_skirt_bk_05_r, swing1LimitDegrees=40, swing2LimitDegrees=40, twistLimitDegrees=5}
 {bone1Name=dyn_skirt_bk_A_04_r, bone2Name=dyn_skirt_bk_04_r, swing1LimitDegrees=40, swing2LimitDegrees=40, twistLimitDegrees=5}
 {bone1Name=dyn_skirt_bk_A_03_r, bone2Name=dyn_skirt_bk_03_r, swing1LimitDegrees=40, swing2LimitDegrees=40, twistLimitDegrees=5}
 {bone1Name=dyn_skirt_frdiag_A_05_l, bone2Name=dyn_skirt_frdiag_05_l, swing1LimitDegrees=45, swing2LimitDegrees=45, twistLimitDegrees=45}
 {bone1Name=dyn_skirt_frdiag_A_04_l, bone2Name=dyn_skirt_frdiag_04_l, swing1LimitDegrees=45, swing2LimitDegrees=45, twistLimitDegrees=45}
 {bone1Name=dyn_skirt_frdiag_A_03_l, bone2Name=dyn_skirt_frdiag_03_l, swing1LimitDegrees=45, swing2LimitDegrees=45, twistLimitDegrees=45}
 {bone1Name=dyn_skirt_frdiag_04_l, bone2Name=dyn_skirt_out_04_l, swing1LimitDegrees=40, swing2LimitDegrees=40, twistLimitDegrees=5}
 {bone1Name=dyn_skirt_bk_04_l, bone2Name=dyn_skirt_out_04_l, swing1LimitDegrees=40, swing2LimitDegrees=40, twistLimitDegrees=5}
 {
  bone1Name=dyn_skirt_fr_01_r
  bone2Name=pelvis
  swing1LimitDegrees=45
  swing1Motion=Free
  swing2LimitDegrees=45
  swing2Motion=Free
  twistLimitDegrees=45
  twistMotion=Free
 }
 {bone1Name=dyn_skirt_fr_02_r, bone2Name=dyn_skirt_fr_01_r, swing1LimitDegrees=45, swing2LimitDegrees=45, twistLimitDegrees=45}
 {bone1Name=dyn_skirt_fr_03_r, bone2Name=dyn_skirt_fr_02_r, swing1LimitDegrees=40, swing2LimitDegrees=40, twistLimitDegrees=5}
 {bone1Name=dyn_skirt_fr_04_r, bone2Name=dyn_skirt_fr_03_r, swing1LimitDegrees=40, swing2LimitDegrees=40, twistLimitDegrees=5}
 {bone1Name=dyn_skirt_fr_05_r, bone2Name=dyn_skirt_fr_04_r, swing1LimitDegrees=40, swing2LimitDegrees=40, twistLimitDegrees=5}
 {bone1Name=dyn_skirt_fr_05_r, bone2Name=dyn_skirt_frdiag_A_04_r, swing1LimitDegrees=40, swing2LimitDegrees=40, twistLimitDegrees=5}
 {bone1Name=dyn_skirt_fr_04_r, bone2Name=dyn_skirt_frdiag_A_03_r, swing1LimitDegrees=40, swing2LimitDegrees=40, twistLimitDegrees=5}
 {
  bone1Name=dyn_skirt_frdiag_A_01_l
  bone2Name=dyn_skirt_frdiag_01_l
  swing1LimitDegrees=45
  swing1Motion=Free
  swing2LimitDegrees=45
  swing2Motion=Free
  twistLimitDegrees=45
  twistMotion=Free
 }
 {bone1Name=dyn_skirt_frdiag_A_02_l, bone2Name=dyn_skirt_frdiag_02_l, swing1LimitDegrees=45, swing2LimitDegrees=45, twistLimitDegrees=45}
 {bone1Name=dyn_skirt_bk_A_05_l, bone2Name=dyn_skirt_bk_05_l, swing1LimitDegrees=40, swing2LimitDegrees=40, twistLimitDegrees=5}
 {
  bone1Name=dyn_skirt_bk_A_04_l
  bone2Name=dyn_skirt_bk_04_l
  swing1LimitDegrees=45
  swing1Motion=Free
  swing2LimitDegrees=45
  swing2Motion=Free
  twistLimitDegrees=45
  twistMotion=Free
 }
 {bone1Name=dyn_skirt_frdiag_A_05_r, bone2Name=dyn_skirt_frdiag_05_r, swing1LimitDegrees=45, swing2LimitDegrees=45, twistLimitDegrees=45}
 {bone1Name=dyn_skirt_frdiag_A_04_r, bone2Name=dyn_skirt_frdiag_04_r, swing1LimitDegrees=45, swing2LimitDegrees=45, twistLimitDegrees=45}
 {bone1Name=dyn_skirt_frdiag_05_l, bone2Name=dyn_skirt_out_05_l, swing1LimitDegrees=45, swing2LimitDegrees=45, twistLimitDegrees=45}
 {bone1Name=dyn_skirt_bk_05_r, bone2Name=dyn_skirt_out_05_r, swing1LimitDegrees=40, swing2LimitDegrees=40, twistLimitDegrees=5}
 {bone1Name=dyn_skirt_bk_04_r, bone2Name=dyn_skirt_out_04_r, swing1LimitDegrees=40, swing2LimitDegrees=40, twistLimitDegrees=5}
 {bone1Name=dyn_skirt_bk_03_r, bone2Name=dyn_skirt_out_03_r, swing1LimitDegrees=40, swing2LimitDegrees=40, twistLimitDegrees=5}
 {bone1Name=dyn_skirt_out_05_r, bone2Name=dyn_skirt_frdiag_05_r, swing1LimitDegrees=40, swing2LimitDegrees=40, twistLimitDegrees=5}
 {bone1Name=dyn_skirt_out_04_r, bone2Name=dyn_skirt_frdiag_04_r, swing1LimitDegrees=40, swing2LimitDegrees=40, twistLimitDegrees=5}
 {bone1Name=dyn_skirt_frdiag_A_03_r, bone2Name=dyn_skirt_frdiag_03_r, swing1LimitDegrees=45, swing2LimitDegrees=45, twistLimitDegrees=45}
 {bone1Name=dyn_skirt_frdiag_A_02_r, bone2Name=dyn_skirt_frdiag_02_r, swing1LimitDegrees=45, swing2LimitDegrees=45, twistLimitDegrees=45}
 {bone1Name=dyn_skirt_frdiag_A_01_r, bone2Name=dyn_skirt_frdiag_01_r, swing1LimitDegrees=20, swing2LimitDegrees=20, twistLimitDegrees=5}
 {bone1Name=dyn_skirt_out_03_r, bone2Name=dyn_skirt_frdiag_03_r, swing1LimitDegrees=40, swing2LimitDegrees=40, twistLimitDegrees=5}
 {bone1Name=dyn_skirt_out_02_r, bone2Name=dyn_skirt_frdiag_02_r, swing1LimitDegrees=40, swing2LimitDegrees=40, twistLimitDegrees=5}
 {bone1Name=dyn_skirt_out_03_l, bone2Name=dyn_skirt_frdiag_03_l, swing1LimitDegrees=40, swing2LimitDegrees=40, twistLimitDegrees=5}
 {bone1Name=dyn_skirt_bk_05_l, bone2Name=dyn_skirt_out_05_l, swing1LimitDegrees=45, swing2LimitDegrees=45, twistLimitDegrees=45}
]
