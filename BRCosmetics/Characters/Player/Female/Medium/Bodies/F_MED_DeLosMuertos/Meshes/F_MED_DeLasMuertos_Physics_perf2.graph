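# Physics asset for the female medium De Las Muertos cosmetic body (perf2 variant), built to simulate its skirt.
# Capsules on pelvis, thighs, calves and feet accompany pelvis-rooted chains of skirt-bone capsules around the body, mostly with limited swing and twist.
# Angularly free cross-constraints tie neighboring chains, mostly lower down, linking the skirt like a sheet.
# Left and right differ: the front-left chain is angularly free, and some lower-left bones carry extra long capsules.
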PhysicsAsset /BRCosmetics/Characters/Player/Female/Medium/Bodies/F_MED_DeLosMuertos/Meshes/F_MED_DeLasMuertos_Physics_perf2
bodies {
 calf_l [{shapeType=Capsule, length=35.375, radius=11.625, center=<-24.7642 -2.0577 1>, rotation=<rot 90 0 0>}]
 calf_r [{shapeType=Capsule, length=35.375, radius=11.625, center=<23.8321 2.6464 -1>, rotation=<rot 90 0 0>}]
 dyn_skirt_bk_01 [{shapeType=Capsule, length=4.5, radius=7.5, shapeName=dyn_skirt_bk_01_capsule, center=<4.75 0 5.5>, rotation=<rot 90 0 0>}]
 dyn_skirt_bk_01_l [{shapeType=Capsule, length=4.5, radius=7.5, shapeName=dyn_skirt_bk_01_l_capsule, center=<4.75 0 5.5>, rotation=<rot 90 0 0>}]
 dyn_skirt_bk_01_r [{shapeType=Capsule, length=4.5, radius=7.5, shapeName=dyn_skirt_bk_01_r_capsule, center=<4.75 0 5.5>, rotation=<rot 90 0 0>}]
 dyn_skirt_bk_02 [{shapeType=Capsule, length=8, radius=10, shapeName=dyn_skirt_bk_02_capsule, center=<4.75 0 10>, rotation=<rot 0 0 90>}]
 dyn_skirt_bk_02_l [{shapeType=Capsule, length=8, radius=10, shapeName=dyn_skirt_bk_02_l_capsule, center=<4.75 0 10>, rotation=<rot 0 0 90>}]
 dyn_skirt_bk_02_r [{shapeType=Capsule, length=8, radius=10, shapeName=dyn_skirt_bk_02_r_capsule, center=<4.75 0 10>, rotation=<rot 0 0 90>}]
 dyn_skirt_bk_03 [{shapeType=Capsule, length=8, radius=10, shapeName=dyn_skirt_bk_03_capsule, center=<4.75 0 10>, rotation=<rot 0 0 90>}]
 dyn_skirt_bk_03_l [{shapeType=Capsule, length=8, radius=10, shapeName=dyn_skirt_bk_03_l_capsule, center=<4.75 0 10>, rotation=<rot 0 0 90>}]
 dyn_skirt_bk_03_r [{shapeType=Capsule, length=8, radius=10, shapeName=dyn_skirt_bk_03_r_capsule, center=<4.75 0 10>, rotation=<rot 0 0 90>}]
 dyn_skirt_bk_04 [{shapeType=Capsule, length=8, radius=10, shapeName=dyn_skirt_bk_04_capsule, center=<4.75 0 10>, rotation=<rot 0 0 90>}]
 dyn_skirt_bk_04_l [{shapeType=Capsule, length=8, radius=10, shapeName=dyn_skirt_bk_04_l_capsule, center=<4.75 0 10>, rotation=<rot 0 0 90>}]
 dyn_skirt_bk_04_r [{shapeType=Capsule, length=8, radius=10, shapeName=dyn_skirt_bk_04_r_capsule, center=<4.75 0 10>, rotation=<rot 0 0 90>}]
 dyn_skirt_bk_05 [{shapeType=Capsule, length=8, radius=10, shapeName=dyn_skirt_bk_05_capsule, center=<4.75 0 10>, rotation=<rot 0 0 90>}]
 dyn_skirt_bk_05_l [{shapeType=Capsule, length=8, radius=10, shapeName=dyn_skirt_bk_05_l_capsule, center=<4.75 0 10>, rotation=<rot 0 0 90>}]
 dyn_skirt_bk_05_r [{shapeType=Capsule, length=8, radius=10, shapeName=dyn_skirt_bk_05_r_capsule, center=<4.75 0 10>, rotation=<rot 0 0 90>}]
 dyn_skirt_bk_A_01_l [{shapeType=Capsule, length=4.5, radius=7.5, shapeName=dyn_skirt_bk_A_01_l_capsule, center=<4.75 0 5.5>, rotation=<rot 90 0 0>}]
 dyn_skirt_bk_A_01_r [{shapeType=Capsule, length=4.5, radius=7.5, shapeName=dyn_skirt_bk_A_01_r_capsule, center=<4.75 0 5.5>, rotation=<rot 90 0 0>}]
 dyn_skirt_bk_A_02_l [{shapeType=Capsule, length=8, radius=10, shapeName=dyn_skirt_bk_A_02_l_capsule, center=<4.75 0 10>, rotation=<rot 0 0 90>}]
 dyn_skirt_bk_A_02_r [{shapeType=Capsule, length=8, radius=10, shapeName=dyn_skirt_bk_A_02_r_capsule, center=<4.75 0 10>, rotation=<rot 0 0 90>}]
 dyn_skirt_bk_A_03_l [{shapeType=Capsule, length=8, radius=10, shapeName=dyn_skirt_bk_A_03_l_capsule, center=<4.75 0 10>, rotation=<rot 0 0 90>}]
 dyn_skirt_bk_A_03_r [{shapeType=Capsule, length=8, radius=10, shapeName=dyn_skirt_bk_A_03_r_capsule, center=<4.75 0 10>, rotation=<rot 0 0 90>}]
 dyn_skirt_bk_A_04_l [{shapeType=Capsule, length=8, radius=10, shapeName=dyn_skirt_bk_A_04_l_capsule, center=<4.75 0 10>, rotation=<rot 0 0 90>}]
 dyn_skirt_bk_A_04_r [{shapeType=Capsule, length=8, radius=10, shapeName=dyn_skirt_bk_A_04_r_capsule, center=<4.75 0 10>, rotation=<rot 0 0 90>}]
 dyn_skirt_bk_A_05_l [{shapeType=Capsule, length=4.5, radius=10, shapeName=dyn_skirt_bk_A_05_l_capsule, center=<4.75 0 10>}]
 dyn_skirt_bk_A_05_r [{shapeType=Capsule, length=8, radius=10, shapeName=dyn_skirt_bk_A_05_r_capsule, center=<4.75 0 10>, rotation=<rot 0 0 90>}]
 dyn_skirt_fr_01 [{shapeType=Capsule, length=4.5, radius=7.5, shapeName=dyn_skirt_fr_01_capsule, center=<4.75 0 5.5>, rotation=<rot 90 0 0>}]
 dyn_skirt_fr_01_l [{shapeType=Capsule, length=4.5, radius=7.5, shapeName=dyn_skirt_fr_01_l_capsule, center=<-4.75 0 5.5>, rotation=<rot 90 0 0>}]
 dyn_skirt_fr_01_r [{shapeType=Capsule, length=4.5, radius=7.5, shapeName=dyn_skirt_fr_01_r_capsule, center=<4.75 0 5.5>, rotation=<rot 90 0 0>}]
 dyn_skirt_fr_02 [{shapeType=Capsule, length=4.5, radius=7.5, shapeName=dyn_skirt_fr_02_capsule, center=<4.75 0 5.5>, rotation=<rot 90 0 0>}]
 dyn_skirt_fr_02_l [{shapeType=Capsule, length=4.5, radius=7.5, shapeName=dyn_skirt_fr_02_l_capsule, center=<-4.75 0 5.5>, rotation=<rot 90 0 0>}]
 dyn_skirt_fr_02_r [{shapeType=Capsule, length=4.5, radius=7.5, shapeName=dyn_skirt_fr_02_r_capsule, center=<4.75 0 5.5>, rotation=<rot 90 0 0>}]
 dyn_skirt_fr_03 [{shapeType=Capsule, length=4.5, radius=7.2, shapeName=dyn_skirt_fr_03_capsule, center=<4.75 0 5.5>, rotation=<rot 90 0 0>}]
 dyn_skirt_fr_03_l [{shapeType=Capsule, length=4.5, radius=7.2, shapeName=dyn_skirt_fr_03_l_capsule, center=<-4.75 0 5.5>, rotation=<rot 90 0 0>}]
 dyn_skirt_fr_03_r [{shapeType=Capsule, length=4.5, radius=7.2, shapeName=dyn_skirt_fr_03_r_capsule, center=<4.75 0 5.5>, rotation=<rot 90 0 0>}]
 dyn_skirt_fr_04 [{shapeType=Capsule, length=4.5, radius=7.35, shapeName=dyn_skirt_fr_04_capsule, center=<4.75 0 5.5>, rotation=<rot 90 0 0>}]
 dyn_skirt_fr_04_l [{shapeType=Capsule, length=4.5, radius=7.35, shapeName=dyn_skirt_fr_04_l_capsule, center=<-4.75 0 5.5>, rotation=<rot 90 0 0>}]
 dyn_skirt_fr_04_r [{shapeType=Capsule, length=4.5, radius=7.35, shapeName=dyn_skirt_fr_04_r_capsule, center=<4.75 0 5.5>, rotation=<rot 90 0 0>}]
 dyn_skirt_fr_05_l [{shapeType=Capsule, length=4.5, radius=7.5, shapeName=dyn_skirt_fr_05_l_capsule, center=<-4.75 0 5.5>, rotation=<rot 90 0 0>}]
 dyn_skirt_fr_05_r [{shapeType=Capsule, length=4.5, radius=7.5, shapeName=dyn_skirt_fr_05_r_capsule, center=<4.75 0 5.5>, rotation=<rot 90 0 0>}]
 dyn_skirt_frdiag_01_l [{shapeType=Capsule, length=4.5, radius=7.5, shapeName=dyn_skirt_frdiag_01_l_capsule, center=<4.75 0 5.5>, rotation=<rot 90 0 0>}]
 dyn_skirt_frdiag_01_r [{shapeType=Capsule, length=4.5, radius=7.5, shapeName=dyn_skirt_frdiag_01_r_capsule, center=<4.75 0 5.5>, rotation=<rot 90 0 0>}]
 dyn_skirt_frdiag_02_l [{shapeType=Capsule, length=8, radius=10, shapeName=dyn_skirt_frdiag_02_l_capsule, center=<4.75 0 10>, rotation=<rot 0 0 90>}]
 dyn_skirt_frdiag_02_r [{shapeType=Capsule, length=4.5, radius=10, shapeName=dyn_skirt_frdiag_02_r_capsule, center=<4.75 0 10>, rotation=<rot 90 0 0>}]
 dyn_skirt_frdiag_03_l [{shapeType=Capsule, length=8, radius=10, shapeName=dyn_skirt_frdiag_03_l_capsule, center=<4.75 0 10>, rotation=<rot 0 0 90>}]
 dyn_skirt_frdiag_03_r [{shapeType=Capsule, length=4.5, radius=10, shapeName=dyn_skirt_frdiag_03_r_capsule, center=<4.75 0 10>, rotation=<rot 90 0 0>}]
 dyn_skirt_frdiag_04_l [
  {shapeType=Capsule, length=8, radius=10, shapeName=dyn_skirt_frdiag_04_l_capsule, center=<4.75 0 10>, rotation=<rot 0 0 90>}
  {shapeType=Capsule, length=27.5, radius=5, shapeName=dyn_skirt_frdiag_05_l_capsule, center=<10.0061 -6.5564 0>, rotation=<rot 90 180 69.9999>}
 ]
 dyn_skirt_frdiag_04_r [{shapeType=Capsule, length=4.5, radius=10, shapeName=dyn_skirt_frdiag_04_r_capsule, center=<4.75 0 10>, rotation=<rot 90 0 0>}]
 dyn_skirt_frdiag_05_l [
  {shapeType=Capsule, length=8, radius=10, shapeName=dyn_skirt_frdiag_05_l_capsule, center=<4.75 0 10>, rotation=<rot 0 0 90>}
  {shapeType=Capsule, length=27.5, radius=5, shapeName=dyn_skirt_frdiag_05_l_capsule, center=<10.0061 -6.5564 0>, rotation=<rot 90 180 69.9999>}
 ]
 dyn_skirt_frdiag_05_r [{shapeType=Capsule, length=4.5, radius=10, shapeName=dyn_skirt_frdiag_05_r_capsule, center=<4.75 0 10>, rotation=<rot 90 0 0>}]
 dyn_skirt_frdiag_A_01_l [{shapeType=Capsule, length=4.5, radius=7.5, shapeName=dyn_skirt_frdiag_A_01_l_capsule, center=<4.75 0 5.5>, rotation=<rot 90 0 0>}]
 dyn_skirt_frdiag_A_01_r [{shapeType=Capsule, length=4.5, radius=7.5, shapeName=dyn_skirt_frdiag_A_01_r_capsule, center=<4.75 0 5.5>, rotation=<rot 90 0 0>}]
 dyn_skirt_frdiag_A_02_l [{shapeType=Capsule, length=4.5, radius=7.5, shapeName=dyn_skirt_frdiag_A_02_l_capsule, center=<4.75 0 5.5>, rotation=<rot 90 0 0>}]
 dyn_skirt_frdiag_A_02_r [{shapeType=Capsule, length=4.5, radius=7.5, shapeName=dyn_skirt_frdiag_A_02_r_capsule, center=<4.75 0 5.5>, rotation=<rot 90 0 0>}]
 dyn_skirt_frdiag_A_03_l [{shapeType=Capsule, length=4.5, radius=7.2, shapeName=dyn_skirt_frdiag_A_03_l_capsule, center=<4.75 0 5.5>, rotation=<rot 90 0 0>}]
 dyn_skirt_frdiag_A_03_r [{shapeType=Capsule, length=4.5, radius=7.2, shapeName=dyn_skirt_frdiag_A_03_r_capsule, center=<4.75 0 5.5>, rotation=<rot 90 0 0>}]
 dyn_skirt_frdiag_A_04_l [{shapeType=Capsule, length=4.5, radius=7.35, shapeName=dyn_skirt_frdiag_A_04_l_capsule, center=<4.75 0 5.5>, rotation=<rot 90 0 0>}]
 dyn_skirt_frdiag_A_04_r [{shapeType=Capsule, length=4.5, radius=7.35, shapeName=dyn_skirt_frdiag_A_04_r_capsule, center=<4.75 0 5.5>, rotation=<rot 90 0 0>}]
 dyn_skirt_frdiag_A_05_l [{shapeType=Capsule, length=4.5, radius=7.5, shapeName=dyn_skirt_frdiag_A_05_l_capsule, center=<4.75 0 5.5>, rotation=<rot 90 0 0>}]
 dyn_skirt_frdiag_A_05_r [{shapeType=Capsule, length=4.5, radius=7.5, shapeName=dyn_skirt_frdiag_A_05_r_capsule, center=<4.75 0 5.5>, rotation=<rot 90 0 0>}]
 dyn_skirt_out_01_l [{shapeType=Capsule, length=4.5, radius=7.5, shapeName=dyn_skirt_out_01_l_capsule, center=<4.75 0 5.5>, rotation=<rot 90 0 0>}]
 dyn_skirt_out_01_r [{shapeType=Capsule, length=4.5, radius=7.5, shapeName=dyn_skirt_out_01_r_capsule, center=<4.75 0 5.5>, rotation=<rot 90 0 0>}]
 dyn_skirt_out_02_l [{shapeType=Capsule, length=8, radius=10, shapeName=dyn_skirt_out_02_l_capsule, center=<4.75 0 10>, rotation=<rot 0 0 90>}]
 dyn_skirt_out_02_r [{shapeType=Capsule, length=8, radius=10, shapeName=dyn_skirt_out_02_r_capsule, center=<4.75 0 10>, rotation=<rot 0 0 90>}]
 dyn_skirt_out_03_l [{shapeType=Capsule, length=8, radius=10, shapeName=dyn_skirt_out_03_l_capsule, center=<4.75 0 10>, rotation=<rot 0 0 90>}]
 dyn_skirt_out_03_r [{shapeType=Capsule, length=8, radius=10, shapeName=dyn_skirt_out_03_r_capsule, center=<4.75 0 10>, rotation=<rot 0 0 90>}]
 dyn_skirt_out_04_l [
  {shapeType=Capsule, length=8, radius=10, shapeName=dyn_skirt_out_04_l_capsule, center=<4.75 0 10>, rotation=<rot 0 0 90>}
  {shapeType=Capsule, length=27.5, radius=5, shapeName=dyn_skirt_frdiag_05_l_capsule, center=<10.0061 6.5564 0>, rotation=<rot 90 180 69.9999>}
 ]
 dyn_skirt_out_04_r [{shapeType=Capsule, length=8, radius=10, shapeName=dyn_skirt_out_04_r_capsule, center=<4.75 0 10>, rotation=<rot 0 0 90>}]
 dyn_skirt_out_05_l [
  {shapeType=Capsule, length=8, radius=10, shapeName=dyn_skirt_out_05_l_capsule, center=<4.75 0 10>, rotation=<rot 0 0 90>}
  {shapeType=Capsule, length=27.5, radius=5, shapeName=dyn_skirt_frdiag_05_l_capsule, center=<10.0061 6.5564 0>, rotation=<rot 90 180 69.9999>}
 ]
 dyn_skirt_out_05_r [{shapeType=Capsule, length=8, radius=10, shapeName=dyn_skirt_out_05_r_capsule, center=<4.75 0 10>, rotation=<rot 0 0 90>}]
 foot_l [{shapeType=Capsule, length=20.747, radius=7.1007, center=<-2.6657 5.2706 -0.542>, rotation=<rot 0 0 -90>}]
 foot_r [{shapeType=Capsule, length=20.1611, radius=7.797, center=<2.6656 -6.3584 0.542>, rotation=<rot 0 0 -90>}]
 pelvis [
  {shapeType=Capsule, length=11.542, radius=6.7144, center=<-2.5931 -1.1459 0>}
  {shapeType=Capsule, length=2.75, radius=7.25, center=<-2.5987 -2.0584 0>}
 ]
 thigh_l [{shapeType=Capsule, length=23.4828, radius=9.3585, center=<-25.2038 -0.4536 1.1649>, rotation=<rot 90 0 0>}]
 thigh_r [{shapeType=Capsule, length=23.4828, radius=9.3585, center=<25.2038 0.4536 -1.1649>, rotation=<rot 90 0 0>}]
}
constraints [
 {
  bone1Name=calf_l
  bone2Name=thigh_l
  swing1LimitDegrees=45
  swing1Motion=Free
  swing2LimitDegrees=45
  swing2Motion=Free
  twistLimitDegrees=45
  twistMotion=Free
 }
 {
  bone1Name=foot_l
  bone2Name=calf_l
  swing1LimitDegrees=45
  swing1Motion=Free
  swing2LimitDegrees=45
  swing2Motion=Free
  twistLimitDegrees=45
  twistMotion=Free
 }
 {
  bone1Name=calf_r
  bone2Name=thigh_r
  swing1LimitDegrees=45
  swing1Motion=Free
  swing2LimitDegrees=45
  swing2Motion=Free
  twistLimitDegrees=45
  twistMotion=Free
 }
 {
  bone1Name=foot_r
  bone2Name=calf_r
  swing1LimitDegrees=45
  swing1Motion=Free
  swing2LimitDegrees=45
  swing2Motion=Free
  twistLimitDegrees=45
  twistMotion=Free
 }
 {
  bone1Name=thigh_l
  bone2Name=pelvis
  swing1LimitDegrees=45
  swing1Motion=Free
  swing2LimitDegrees=45
  swing2Motion=Free
  twistLimitDegrees=45
  twistMotion=Free
 }
 {
  bone1Name=thigh_r
  bone2Name=pelvis
  swing1LimitDegrees=45
  swing1Motion=Free
  swing2LimitDegrees=45
  swing2Motion=Free
  twistLimitDegrees=45
  twistMotion=Free
 }
 {bone1Name=dyn_skirt_frdiag_01_l, bone2Name=pelvis, swing1LimitDegrees=5, swing2LimitDegrees=10, twistLimitDegrees=2}
 {bone1Name=dyn_skirt_frdiag_02_l, bone2Name=dyn_skirt_frdiag_01_l, swing1LimitDegrees=2, swing2LimitDegrees=40, twistLimitDegrees=2}
 {bone1Name=dyn_skirt_frdiag_03_l, bone2Name=dyn_skirt_frdiag_02_l, swing1LimitDegrees=10, swing2LimitDegrees=40, twistLimitDegrees=5}
 {bone1Name=dyn_skirt_frdiag_04_l, bone2Name=dyn_skirt_frdiag_03_l, swing1LimitDegrees=20, swing2LimitDegrees=40, twistLimitDegrees=5}
 {bone1Name=dyn_skirt_frdiag_05_l, bone2Name=dyn_skirt_frdiag_04_l, swing1LimitDegrees=30, swing2LimitDegrees=40, twistLimitDegrees=5}
 {bone1Name=dyn_skirt_out_01_l, bone2Name=pelvis, swing1LimitDegrees=40, swing2LimitDegrees=40, twistLimitDegrees=5}
 {bone1Name=dyn_skirt_out_02_l, bone2Name=dyn_skirt_out_01_l, swing1LimitDegrees=40, swing2LimitDegrees=40, twistLimitDegrees=5}
 {bone1Name=dyn_skirt_out_03_l, bone2Name=dyn_skirt_out_02_l, swing1LimitDegrees=40, swing2LimitDegrees=40, twistLimitDegrees=5}
 {bone1Name=dyn_skirt_out_04_l, bone2Name=dyn_skirt_out_03_l, swing1LimitDegrees=40, swing2LimitDegrees=40, twistLimitDegrees=5}
 {bone1Name=dyn_skirt_out_05_l, bone2Name=dyn_skirt_out_04_l, swing1LimitDegrees=40, swing2LimitDegrees=40, twistLimitDegrees=5}
 {bone1Name=dyn_skirt_bk_01_l, bone2Name=pelvis, swing1LimitDegrees=40, swing2LimitDegrees=40, twistLimitDegrees=5}
 {bone1Name=dyn_skirt_bk_02_l, bone2Name=dyn_skirt_bk_01_l, swing1LimitDegrees=40, swing2LimitDegrees=40, twistLimitDegrees=5}
 {bone1Name=dyn_skirt_bk_03_l, bone2Name=dyn_skirt_bk_02_l, swing1LimitDegrees=40, swing2LimitDegrees=40, twistLimitDegrees=5}
 {bone1Name=dyn_skirt_bk_04_l, bone2Name=dyn_skirt_bk_03_l, swing1LimitDegrees=40, swing2LimitDegrees=40, twistLimitDegrees=5}
 {bone1Name=dyn_skirt_bk_05_l, bone2Name=dyn_skirt_bk_04_l, swing1LimitDegrees=40, swing2LimitDegrees=40, twistLimitDegrees=5}
 {bone1Name=dyn_skirt_fr_01, bone2Name=pelvis, swing1LimitDegrees=40, swing2LimitDegrees=40, twistLimitDegrees=5}
 {bone1Name=dyn_skirt_fr_02, bone2Name=dyn_skirt_fr_01, swing1LimitDegrees=40, swing2LimitDegrees=40, twistLimitDegrees=5}
 {bone1Name=dyn_skirt_fr_03, bone2Name=dyn_skirt_fr_02, swing1LimitDegrees=40, swing2LimitDegrees=40, twistLimitDegrees=5}
 {bone1Name=dyn_skirt_fr_04, bone2Name=dyn_skirt_fr_03, swing1LimitDegrees=40, swing2LimitDegrees=40, twistLimitDegrees=5}
 {bone1Name=dyn_skirt_bk_01, bone2Name=pelvis, swing1LimitDegrees=40, swing2LimitDegrees=40, twistLimitDegrees=5}
 {bone1Name=dyn_skirt_bk_02, bone2Name=dyn_skirt_bk_01, swing1LimitDegrees=40, swing2LimitDegrees=40, twistLimitDegrees=5}
 {bone1Name=dyn_skirt_bk_03, bone2Name=dyn_skirt_bk_02, swing1LimitDegrees=40, swing2LimitDegrees=40, twistLimitDegrees=5}
 {bone1Name=dyn_skirt_bk_04, bone2Name=dyn_skirt_bk_03, swing1LimitDegrees=40, swing2LimitDegrees=40, twistLimitDegrees=5}
 {bone1Name=dyn_skirt_bk_05, bone2Name=dyn_skirt_bk_04, swing1LimitDegrees=40, swing2LimitDegrees=40, twistLimitDegrees=5}
 {bone1Name=dyn_skirt_frdiag_01_r, bone2Name=pelvis, swing1LimitDegrees=5, swing2LimitDegrees=40, twistLimitDegrees=2}
 {bone1Name=dyn_skirt_frdiag_02_r, bone2Name=dyn_skirt_frdiag_01_r, swing1LimitDegrees=2, swing2LimitDegrees=40, twistLimitDegrees=2}
 {bone1Name=dyn_skirt_frdiag_03_r, bone2Name=dyn_skirt_frdiag_02_r, swing1LimitDegrees=10, swing2LimitDegrees=40, twistLimitDegrees=5}
 {bone1Name=dyn_skirt_frdiag_04_r, bone2Name=dyn_skirt_frdiag_03_r, swing1LimitDegrees=20, swing2LimitDegrees=40, twistLimitDegrees=5}
 {bone1Name=dyn_skirt_frdiag_05_r, bone2Name=dyn_skirt_frdiag_04_r, swing1LimitDegrees=30, swing2LimitDegrees=40, twistLimitDegrees=5}
 {bone1Name=dyn_skirt_out_01_r, bone2Name=pelvis, swing1LimitDegrees=40, swing2LimitDegrees=40, twistLimitDegrees=5}
 {bone1Name=dyn_skirt_out_02_r, bone2Name=dyn_skirt_out_01_r, swing1LimitDegrees=40, swing2LimitDegrees=40, twistLimitDegrees=5}
 {bone1Name=dyn_skirt_out_03_r, bone2Name=dyn_skirt_out_02_r, swing1LimitDegrees=40, swing2LimitDegrees=40, twistLimitDegrees=5}
 {bone1Name=dyn_skirt_out_04_r, bone2Name=dyn_skirt_out_03_r, swing1LimitDegrees=40, swing2LimitDegrees=40, twistLimitDegrees=5}
 {bone1Name=dyn_skirt_out_05_r, bone2Name=dyn_skirt_out_04_r, swing1LimitDegrees=40, swing2LimitDegrees=40, twistLimitDegrees=5}
 {bone1Name=dyn_skirt_bk_01_r, bone2Name=pelvis, swing1LimitDegrees=40, swing2LimitDegrees=40, twistLimitDegrees=5}
 {bone1Name=dyn_skirt_bk_02_r, bone2Name=dyn_skirt_bk_01_r, swing1LimitDegrees=40, swing2LimitDegrees=40, twistLimitDegrees=5}
 {bone1Name=dyn_skirt_bk_03_r, bone2Name=dyn_skirt_bk_02_r, swing1LimitDegrees=40, swing2LimitDegrees=40, twistLimitDegrees=5}
 {bone1Name=dyn_skirt_bk_04_r, bone2Name=dyn_skirt_bk_03_r, swing1LimitDegrees=40, swing2LimitDegrees=40, twistLimitDegrees=5}
 {bone1Name=dyn_skirt_bk_05_r, bone2Name=dyn_skirt_bk_04_r, swing1LimitDegrees=40, swing2LimitDegrees=40, twistLimitDegrees=5}
 {bone1Name=dyn_skirt_frdiag_A_01_r, bone2Name=pelvis, swing1LimitDegrees=5, swing2LimitDegrees=40, twistLimitDegrees=1}
 {bone1Name=dyn_skirt_frdiag_A_02_r, bone2Name=dyn_skirt_frdiag_A_01_r, swing1LimitDegrees=5, swing2LimitDegrees=40, twistLimitDegrees=1}
 {bone1Name=dyn_skirt_frdiag_A_03_r, bone2Name=dyn_skirt_frdiag_A_02_r, swing1LimitDegrees=10, swing2LimitDegrees=40, twistLimitDegrees=5}
 {bone1Name=dyn_skirt_frdiag_A_04_r, bone2Name=dyn_skirt_frdiag_A_03_r, swing1LimitDegrees=20, swing2LimitDegrees=40, twistLimitDegrees=5}
 {bone1Name=dyn_skirt_frdiag_A_05_r, bone2Name=dyn_skirt_frdiag_A_04_r, swing1LimitDegrees=30, swing2LimitDegrees=40, twistLimitDegrees=5}
 {bone1Name=dyn_skirt_frdiag_A_01_l, bone2Name=pelvis, swing1LimitDegrees=40, swing2LimitDegrees=40, twistLimitDegrees=40}
 {bone1Name=dyn_skirt_frdiag_A_02_l, bone2Name=dyn_skirt_frdiag_A_01_l, swing1LimitDegrees=5, swing2LimitDegrees=40, twistLimitDegrees=1}
 {bone1Name=dyn_skirt_frdiag_A_03_l, bone2Name=dyn_skirt_frdiag_A_02_l, swing1LimitDegrees=5, swing2LimitDegrees=40, twistLimitDegrees=5}
 {bone1Name=dyn_skirt_frdiag_A_04_l, bone2Name=dyn_skirt_frdiag_A_03_l, swing1LimitDegrees=30, swing2LimitDegrees=40, twistLimitDegrees=5}
 {bone1Name=dyn_skirt_frdiag_A_05_l, bone2Name=dyn_skirt_frdiag_A_04_l, swing1LimitDegrees=30.7143, swing2LimitDegrees=40, twistLimitDegrees=5}
 {bone1Name=dyn_skirt_bk_A_01_r, bone2Name=pelvis, swing1LimitDegrees=40, swing2LimitDegrees=40, twistLimitDegrees=5}
 {bone1Name=dyn_skirt_bk_A_02_r, bone2Name=dyn_skirt_bk_A_01_r, swing1LimitDegrees=40, swing2LimitDegrees=40, twistLimitDegrees=5}
 {bone1Name=dyn_skirt_bk_A_03_r, bone2Name=dyn_skirt_bk_A_02_r, swing1LimitDegrees=40, swing2LimitDegrees=40, twistLimitDegrees=5}
 {bone1Name=dyn_skirt_bk_A_04_r, bone2Name=dyn_skirt_bk_A_03_r, swing1LimitDegrees=40, swing2LimitDegrees=40, twistLimitDegrees=5}
 {bone1Name=dyn_skirt_bk_A_05_r, bone2Name=dyn_skirt_bk_A_04_r, swing1LimitDegrees=40, swing2LimitDegrees=40, twistLimitDegrees=5}
 {bone1Name=dyn_skirt_bk_A_01_l, bone2Name=pelvis, swing1LimitDegrees=40, swing2LimitDegrees=40, twistLimitDegrees=5}
 {bone1Name=dyn_skirt_bk_A_02_l, bone2Name=dyn_skirt_bk_A_01_l, swing1LimitDegrees=40, swing2LimitDegrees=40, twistLimitDegrees=5}
 {bone1Name=dyn_skirt_bk_A_03_l, bone2Name=dyn_skirt_bk_A_02_l, swing1LimitDegrees=40, swing2LimitDegrees=40, twistLimitDegrees=5}
 {bone1Name=dyn_skirt_bk_A_04_l, bone2Name=dyn_skirt_bk_A_03_l, swing1LimitDegrees=40, swing2LimitDegrees=40, twistLimitDegrees=5}
 {bone1Name=dyn_skirt_bk_A_05_l, bone2Name=dyn_skirt_bk_A_04_l, swing1LimitDegrees=40, swing2LimitDegrees=40, twistLimitDegrees=5}
 {
  bone1Name=dyn_skirt_bk_A_03_l
  bone2Name=dyn_skirt_bk_03
  swing1LimitDegrees=45
  swing1Motion=Free
  swing2LimitDegrees=45
  swing2Motion=Free
  twistLimitDegrees=15
  twistMotion=Free
 }
 {
  bone1Name=dyn_skirt_bk_A_03_r
  bone2Name=dyn_skirt_bk_03
  swing1LimitDegrees=45
  swing1Motion=Free
  swing2LimitDegrees=45
  swing2Motion=Free
  twistLimitDegrees=15
  twistMotion=Free
 }
 {
  bone1Name=dyn_skirt_bk_A_04_l
  bone2Name=dyn_skirt_bk_04
  swing1LimitDegrees=45
  swing1Motion=Free
  swing2LimitDegrees=45
  swing2Motion=Free
  twistLimitDegrees=15
  twistMotion=Free
 }
 {
  bone1Name=dyn_skirt_bk_A_04_r
  bone2Name=dyn_skirt_bk_04
  swing1LimitDegrees=45
  swing1Motion=Free
  swing2LimitDegrees=45
  swing2Motion=Free
  twistLimitDegrees=15
  twistMotion=Free
 }
 {
  bone1Name=dyn_skirt_bk_A_05_l
  bone2Name=dyn_skirt_bk_05
  swing1LimitDegrees=45
  swing1Motion=Free
  swing2LimitDegrees=45
  swing2Motion=Free
  twistLimitDegrees=15
  twistMotion=Free
 }
 {
  bone1Name=dyn_skirt_bk_A_05_r
  bone2Name=dyn_skirt_bk_05
  swing1LimitDegrees=45
  swing1Motion=Free
  swing2LimitDegrees=45
  swing2Motion=Free
  twistLimitDegrees=15
  twistMotion=Free
 }
 {
  bone1Name=dyn_skirt_bk_A_05_r
  bone2Name=dyn_skirt_bk_05_r
  swing1LimitDegrees=45
  swing1Motion=Free
  swing2LimitDegrees=45
  swing2Motion=Free
  twistLimitDegrees=15
  twistMotion=Free
 }
 {
  bone1Name=dyn_skirt_bk_A_04_r
  bone2Name=dyn_skirt_bk_04_r
  swing1LimitDegrees=45
  swing1Motion=Free
  swing2LimitDegrees=45
  swing2Motion=Free
  twistLimitDegrees=15
  twistMotion=Free
 }
 {
  bone1Name=dyn_skirt_bk_A_03_r
  bone2Name=dyn_skirt_bk_03_r
  swing1LimitDegrees=45
  swing1Motion=Free
  swing2LimitDegrees=45
  swing2Motion=Free
  twistLimitDegrees=15
  twistMotion=Free
 }
 {
  bone1Name=dyn_skirt_frdiag_A_05_l
  bone2Name=dyn_skirt_frdiag_05_l
  swing1LimitDegrees=45
  swing1Motion=Free
  swing2LimitDegrees=45
  swing2Motion=Free
  twistLimitDegrees=45
  twistMotion=Free
 }
 {
  bone1Name=dyn_skirt_frdiag_A_04_l
  bone2Name=dyn_skirt_frdiag_04_l
  swing1LimitDegrees=45
  swing1Motion=Free
  swing2LimitDegrees=45
  swing2Motion=Free
  twistLimitDegrees=45
  twistMotion=Free
 }
 {
  bone1Name=dyn_skirt_frdiag_A_03_l
  bone2Name=dyn_skirt_frdiag_03_l
  swing1LimitDegrees=45
  swing1Motion=Free
  swing2LimitDegrees=45
  swing2Motion=Free
  twistLimitDegrees=45
  twistMotion=Free
 }
 {
  bone1Name=dyn_skirt_frdiag_04_l
  bone2Name=dyn_skirt_out_04_l
  swing1LimitDegrees=40
  swing1Motion=Free
  swing2LimitDegrees=40
  swing2Motion=Free
  twistLimitDegrees=5
  twistMotion=Free
 }
 {
  bone1Name=dyn_skirt_bk_04_l
  bone2Name=dyn_skirt_out_04_l
  swing1LimitDegrees=40
  swing1Motion=Free
  swing2LimitDegrees=40
  swing2Motion=Free
  twistLimitDegrees=5
  twistMotion=Free
 }
 {
  bone1Name=dyn_skirt_fr_01_r
  bone2Name=pelvis
  swing1LimitDegrees=45
  swing1Motion=Free
  swing2LimitDegrees=45
  swing2Motion=Free
  twistLimitDegrees=45
  twistMotion=Free
 }
 {bone1Name=dyn_skirt_fr_02_r, bone2Name=dyn_skirt_fr_01_r, swing1LimitDegrees=45, swing2LimitDegrees=45, twistLimitDegrees=45}
 {bone1Name=dyn_skirt_fr_03_r, bone2Name=dyn_skirt_fr_02_r, swing1LimitDegrees=40, swing2LimitDegrees=40, twistLimitDegrees=5}
 {bone1Name=dyn_skirt_fr_04_r, bone2Name=dyn_skirt_fr_03_r, swing1LimitDegrees=40, swing2LimitDegrees=40, twistLimitDegrees=5}
 {bone1Name=dyn_skirt_fr_05_r, bone2Name=dyn_skirt_fr_04_r, swing1LimitDegrees=40, swing2LimitDegrees=40, twistLimitDegrees=5}
 {
  bone1Name=dyn_skirt_fr_05_r
  bone2Name=dyn_skirt_frdiag_A_04_r
  swing1LimitDegrees=40
  swing1Motion=Free
  swing2LimitDegrees=40
  swing2Motion=Free
  twistLimitDegrees=5
  twistMotion=Free
 }
 {
  bone1Name=dyn_skirt_fr_04_r
  bone2Name=dyn_skirt_frdiag_A_03_r
  swing1LimitDegrees=40
  swing1Motion=Free
  swing2LimitDegrees=40
  swing2Motion=Free
  twistLimitDegrees=5
  twistMotion=Free
 }
 {
  bone1Name=dyn_skirt_frdiag_A_01_l
  bone2Name=dyn_skirt_frdiag_01_l
  swing1LimitDegrees=45
  swing1Motion=Free
  swing2LimitDegrees=45
  swing2Motion=Free
  twistLimitDegrees=45
  twistMotion=Free
 }
 {
  bone1Name=dyn_skirt_frdiag_A_02_l
  bone2Name=dyn_skirt_frdiag_02_l
  swing1LimitDegrees=45
  swing1Motion=Free
  swing2LimitDegrees=45
  swing2Motion=Free
  twistLimitDegrees=45
  twistMotion=Free
 }
 {
  bone1Name=dyn_skirt_bk_A_05_l
  bone2Name=dyn_skirt_bk_05_l
  swing1LimitDegrees=45
  swing1Motion=Free
  swing2LimitDegrees=45
  swing2Motion=Free
  twistLimitDegrees=15
  twistMotion=Free
 }
 {
  bone1Name=dyn_skirt_bk_A_04_l
  bone2Name=dyn_skirt_bk_04_l
  swing1LimitDegrees=45
  swing1Motion=Free
  swing2LimitDegrees=45
  swing2Motion=Free
  twistLimitDegrees=15
  twistMotion=Free
 }
 {
  bone1Name=dyn_skirt_frdiag_A_05_r
  bone2Name=dyn_skirt_frdiag_05_r
  swing1LimitDegrees=45
  swing1Motion=Free
  swing2LimitDegrees=45
  swing2Motion=Free
  twistLimitDegrees=45
  twistMotion=Free
 }
 {
  bone1Name=dyn_skirt_frdiag_A_04_r
  bone2Name=dyn_skirt_frdiag_04_r
  swing1LimitDegrees=45
  swing1Motion=Free
  swing2LimitDegrees=45
  swing2Motion=Free
  twistLimitDegrees=45
  twistMotion=Free
 }
 {
  bone1Name=dyn_skirt_frdiag_05_l
  bone2Name=dyn_skirt_out_05_l
  swing1LimitDegrees=45
  swing1Motion=Free
  swing2LimitDegrees=45
  swing2Motion=Free
  twistLimitDegrees=45
  twistMotion=Free
 }
 {
  bone1Name=dyn_skirt_bk_05_r
  bone2Name=dyn_skirt_out_05_r
  swing1LimitDegrees=40
  swing1Motion=Free
  swing2LimitDegrees=40
  swing2Motion=Free
  twistLimitDegrees=5
  twistMotion=Free
 }
 {
  bone1Name=dyn_skirt_bk_04_r
  bone2Name=dyn_skirt_out_04_r
  swing1LimitDegrees=40
  swing1Motion=Free
  swing2LimitDegrees=40
  swing2Motion=Free
  twistLimitDegrees=5
  twistMotion=Free
 }
 {
  bone1Name=dyn_skirt_out_05_r
  bone2Name=dyn_skirt_frdiag_05_r
  swing1LimitDegrees=40
  swing1Motion=Free
  swing2LimitDegrees=40
  swing2Motion=Free
  twistLimitDegrees=5
  twistMotion=Free
 }
 {
  bone1Name=dyn_skirt_out_04_r
  bone2Name=dyn_skirt_frdiag_04_r
  swing1LimitDegrees=40
  swing1Motion=Free
  swing2LimitDegrees=40
  swing2Motion=Free
  twistLimitDegrees=5
  twistMotion=Free
 }
 {
  bone1Name=dyn_skirt_frdiag_A_03_r
  bone2Name=dyn_skirt_frdiag_03_r
  swing1LimitDegrees=45
  swing1Motion=Free
  swing2LimitDegrees=45
  swing2Motion=Free
  twistLimitDegrees=45
  twistMotion=Free
 }
 {
  bone1Name=dyn_skirt_frdiag_A_02_r
  bone2Name=dyn_skirt_frdiag_02_r
  swing1LimitDegrees=45
  swing1Motion=Free
  swing2LimitDegrees=45
  swing2Motion=Free
  twistLimitDegrees=45
  twistMotion=Free
 }
 {
  bone1Name=dyn_skirt_frdiag_A_01_r
  bone2Name=dyn_skirt_frdiag_01_r
  swing1LimitDegrees=20
  swing1Motion=Free
  swing2LimitDegrees=20
  swing2Motion=Free
  twistLimitDegrees=5
  twistMotion=Free
 }
 {
  bone1Name=dyn_skirt_out_03_r
  bone2Name=dyn_skirt_frdiag_03_r
  swing1LimitDegrees=40
  swing1Motion=Free
  swing2LimitDegrees=40
  swing2Motion=Free
  twistLimitDegrees=5
  twistMotion=Free
 }
 {
  bone1Name=dyn_skirt_out_02_r
  bone2Name=dyn_skirt_frdiag_02_r
  swing1LimitDegrees=40
  swing1Motion=Free
  swing2LimitDegrees=40
  swing2Motion=Free
  twistLimitDegrees=5
  twistMotion=Free
 }
 {
  bone1Name=dyn_skirt_out_03_l
  bone2Name=dyn_skirt_frdiag_03_l
  swing1LimitDegrees=40
  swing1Motion=Free
  swing2LimitDegrees=40
  swing2Motion=Free
  twistLimitDegrees=5
  twistMotion=Free
 }
 {
  bone1Name=dyn_skirt_bk_05_l
  bone2Name=dyn_skirt_out_05_l
  swing1LimitDegrees=45
  swing1Motion=Free
  swing2LimitDegrees=45
  swing2Motion=Free
  twistLimitDegrees=45
  twistMotion=Free
 }
 {
  bone1Name=dyn_skirt_fr_01_l
  bone2Name=pelvis
  swing1LimitDegrees=45
  swing1Motion=Free
  swing2LimitDegrees=45
  swing2Motion=Free
  twistLimitDegrees=45
  twistMotion=Free
 }
 {
  bone1Name=dyn_skirt_fr_02_l
  bone2Name=dyn_skirt_fr_01_l
  swing1LimitDegrees=45
  swing1Motion=Free
  swing2LimitDegrees=45
  swing2Motion=Free
  twistLimitDegrees=45
  twistMotion=Free
 }
 {
  bone1Name=dyn_skirt_fr_03_l
  bone2Name=dyn_skirt_fr_02_l
  swing1LimitDegrees=45
  swing1Motion=Free
  swing2LimitDegrees=45
  swing2Motion=Free
  twistLimitDegrees=45
  twistMotion=Free
 }
 {
  bone1Name=dyn_skirt_fr_04_l
  bone2Name=dyn_skirt_fr_03_l
  swing1LimitDegrees=45
  swing1Motion=Free
  swing2LimitDegrees=45
  swing2Motion=Free
  twistLimitDegrees=45
  twistMotion=Free
 }
 {
  bone1Name=dyn_skirt_fr_05_l
  bone2Name=dyn_skirt_fr_04_l
  swing1LimitDegrees=45
  swing1Motion=Free
  swing2LimitDegrees=45
  swing2Motion=Free
  twistLimitDegrees=45
  twistMotion=Free
 }
 {
  bone1Name=dyn_skirt_frdiag_A_03_l
  bone2Name=dyn_skirt_fr_04_l
  swing1LimitDegrees=40
  swing1Motion=Free
  swing2LimitDegrees=40
  swing2Motion=Free
  twistLimitDegrees=5
  twistMotion=Free
 }
 {
  bone1Name=dyn_skirt_frdiag_A_04_l
  bone2Name=dyn_skirt_fr_05_l
  swing1LimitDegrees=40
  swing1Motion=Free
  swing2LimitDegrees=40
  swing2Motion=Free
  twistLimitDegrees=5
  twistMotion=Free
 }
 {
  bone1Name=dyn_skirt_bk_03_r
  bone2Name=dyn_skirt_out_03_r
  swing1LimitDegrees=40
  swing1Motion=Free
  swing2LimitDegrees=40
  swing2Motion=Free
  twistLimitDegrees=5
  twistMotion=Free
 }
 {
  bone1Name=dyn_skirt_bk_03_l
  bone2Name=dyn_skirt_out_03_l
  swing1LimitDegrees=40
  swing1Motion=Free
  swing2LimitDegrees=40
  swing2Motion=Free
  twistLimitDegrees=5
  twistMotion=Free
 }
]
