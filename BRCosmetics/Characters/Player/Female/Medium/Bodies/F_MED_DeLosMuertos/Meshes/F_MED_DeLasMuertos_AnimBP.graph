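# Animation Blueprint for the De Las Muertos female cosmetic body, adding procedural secondary motion to its flower, rose, bows and skirt.
# Chained AnimDynamics nodes simulate each dyn_ bone, some gated by alphaVal.
# Skirt bones simulate relative to the thighs, using spherical limits against leg clipping and, on many, a skydiving force.
# Additive bone offsets driven by crouch and right-leg alphas, plus thigh constraints, adjust poses before simulation.
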AnimBlueprint /BRCosmetics/Characters/Player/Female/Medium/Bodies/F_MED_DeLosMuertos/Meshes/F_MED_DeLasMuertos_AnimBP
target_skeleton=&/F_MED_DeLasMuertos_Skeleton
graphs [
 AnimGraph {nodes [
   n1 = Anim Dynamics - Bone: dyn_collar_flower(Alpha=1, ExternalForce="0,0,0") {
    angularDampingOverride=0.99
    angularSpringConstant=1500
    bAngularSpring=true
    bGravityOverrideInSimSpace=true
    bOverrideAngularDamping=true
    gravityScale=0.35
    numSolverIterationsPostUpdate=2
    numSolverIterationsPreUpdate=8
    simulationSpace=RootRelative
    alphaBoolBlend {blendOption=Linear}
    boundBone {boneName=dyn_collar_flower}
    componentAppliedLinearAccClamp=<3000 3000 3000>
    componentLinearAccScale=<0.2 0.2 0.2>
    componentLinearVelScale=<0.2 0.2 0.2>
    physicsBodyDefinitions [{
      sphereCollisionRadius=0
      boundBone {boneName=dyn_collar_flower}
      boxExtents=<20 20 10>
      constraintSetup {angularTargetAxis=AxisZ, twistAxis=AxisZ, angularLimitsMax=<5 3 5>, angularLimitsMin=<-5 -3 -5>, angularTarget=<0 0 0>}
      localJointOffset=<0 0 -2>
     }]
   }
   n2 = "Transform (Modify) Bone - Bone: dyn_hip_rose"(ComponentPose=@n1, Alpha=Get R Leg Fwd Alpha() -> (R Leg Fwd Alpha:double), Translation="0,-1,5", Rotation="0,20,-65", Scale="1,1,1") {
    rotationMode=BMM_Additive
    rotationSpace=BCS_BoneSpace
    translationMode=BMM_Additive
    translationSpace=BCS_BoneSpace
    alphaBoolBlend {blendOption=Linear}
    boneToModify {boneName=dyn_hip_rose}
    rotation {roll=-65, yaw=20}
    translation {y=-1, z=5}
   }
   n3 = "Transform (Modify) Bone - Bone: dyn_hip_rose"(ComponentPose=@n2, Alpha=Get R Leg Side Alpha() -> (R Leg Side Alpha:double), Translation="-4,0,0", Rotation="0,20,0", Scale="1,1,1") {
    rotationMode=BMM_Additive
    rotationSpace=BCS_BoneSpace
    translationMode=BMM_Additive
    translationSpace=BCS_BoneSpace
    alphaBoolBlend {blendOption=Linear}
    boneToModify {boneName=dyn_hip_rose}
    rotation {yaw=20}
    translation {x=-4}
   }
   n4 = Get alphaVal() -> (alphaVal:double)
   n5 = Anim Dynamics - Bone: dyn_hip_rose(ComponentPose=@n3, Alpha=@n4, ExternalForce="0,0,0") {
    angularDampingOverride=0.99
    angularSpringConstant=1500
    bAngularSpring=true
    bGravityOverrideInSimSpace=true
    bOverrideAngularDamping=true
    gravityScale=0.35
    numSolverIterationsPostUpdate=2
    numSolverIterationsPreUpdate=8
    simulationSpace=RootRelative
    alphaBoolBlend {blendOption=Linear}
    boundBone {boneName=dyn_hip_rose}
    componentAppliedLinearAccClamp=<3000 3000 3000>
    componentLinearAccScale=<0.2 0.2 0.2>
    componentLinearVelScale=<0.2 0.2 0.2>
    physicsBodyDefinitions [{
      sphereCollisionRadius=0
      boundBone {boneName=dyn_hip_rose}
      boxExtents=<20 20 10>
      constraintSetup {angularTargetAxis=AxisZ, twistAxis=AxisZ, angularLimitsMax=<3 3 3>, angularLimitsMin=<-3 -3 -3>, angularTarget=<0 0 0>}
      localJointOffset=<0 0 -2>
     }]
   }
   n6 = Anim Dynamics - Bone: dyn_back_bow(ComponentPose=@n5, Alpha=1, ExternalForce="0,0,0") {
    angularDampingOverride=0.99
    angularSpringConstant=1500
    bAngularSpring=true
    bGravityOverrideInSimSpace=true
    bOverrideAngularDamping=true
    gravityScale=0.35
    numSolverIterationsPostUpdate=2
    numSolverIterationsPreUpdate=8
    simulationSpace=RootRelative
    alphaBoolBlend {blendOption=Linear}
    boundBone {boneName=dyn_back_bow}
    componentAppliedLinearAccClamp=<3000 3000 3000>
    componentLinearAccScale=<0.5 0.5 0.5>
    componentLinearVelScale=<0.5 0.5 0.5>
    physicsBodyDefinitions [{
      sphereCollisionRadius=0
      boundBone {boneName=dyn_back_bow}
      boxExtents=<20 20 5>
      constraintSetup {angularTargetAxis=AxisZ, twistAxis=AxisZ, angularLimitsMax=<-5 5 3>, angularLimitsMin=<-10 -5 -3>, angularTarget=<0 0 0>}
      localJointOffset=<0 0 -2>
     }]
   }
   n7 = Anim Dynamics - Bone: dyn_back_bow_string01_l(ComponentPose=@n6, Alpha=@n4, ExternalForce="0,0,0") {
    angularDampingOverride=0.999
    bGravityOverrideInSimSpace=true
    bOverrideAngularDamping=true
    gravityScale=0.45
    numSolverIterationsPostUpdate=2
    numSolverIterationsPreUpdate=8
    simulationSpace=RootRelative
    alphaBoolBlend {blendOption=Linear}
    boundBone {boneName=dyn_back_bow_string01_l}
    componentAppliedLinearAccClamp=<3000 3000 3000>
    componentLinearAccScale=<0.2 0.2 0.3>
    componentLinearVelScale=<0.2 0.2 0.3>
    physicsBodyDefinitions [{
      sphereCollisionRadius=0
      boundBone {boneName=dyn_back_bow_string01_l}
      boxExtents=<15 6 15>
      constraintSetup {angularTargetAxis=AxisZ, twistAxis=AxisZ, angularLimitsMax=<3 10 3>, angularLimitsMin=<-15 -10 -3>, angularTarget=<0 0 0>}
      localJointOffset=<0 0 -2>
     }]
    relativeSpaceBone {boneName=dyn_back_bow}
   }
   n8 = Anim Dynamics - Bone: dyn_back_bow_string02_l(ComponentPose=@n7, Alpha=@n4, ExternalForce="0,0,0") {
    angularDampingOverride=0.985
    bGravityOverrideInSimSpace=true
    bOverrideAngularDamping=true
    gravityScale=0.35
    numSolverIterationsPostUpdate=2
    numSolverIterationsPreUpdate=8
    simulationSpace=RootRelative
    alphaBoolBlend {blendOption=Linear}
    boundBone {boneName=dyn_back_bow_string02_l}
    componentAppliedLinearAccClamp=<3000 3000 3000>
    physicsBodyDefinitions [{
      sphereCollisionRadius=0
      boundBone {boneName=dyn_back_bow_string02_l}
      boxExtents=<5 5 2>
      constraintSetup {angularTargetAxis=AxisZ, twistAxis=AxisZ, angularLimitsMax=<2 5 5>, angularLimitsMin=<-7 -5 -5>, angularTarget=<0 0 0>}
      localJointOffset=<0 0 -4.8>
     }]
   }
   n9 = Anim Dynamics - Bone: dyn_back_bow_string01_r(ComponentPose=@n8, Alpha=@n4, ExternalForce="0,0,0") {
    angularDampingOverride=0.987
    bGravityOverrideInSimSpace=true
    bOverrideAngularDamping=true
    gravityScale=0.35
    numSolverIterationsPostUpdate=2
    numSolverIterationsPreUpdate=8
    simulationSpace=RootRelative
    alphaBoolBlend {blendOption=Linear}
    boundBone {boneName=dyn_back_bow_string01_r}
    componentAppliedLinearAccClamp=<3000 3000 3000>
    componentLinearAccScale=<0.2 0.2 0.2>
    componentLinearVelScale=<0.2 0.2 0.2>
    physicsBodyDefinitions [{
      sphereCollisionRadius=0
      boundBone {boneName=dyn_back_bow_string01_r}
      boxExtents=<5 5 2>
      constraintSetup {angularTargetAxis=AxisZ, twistAxis=AxisZ, angularLimitsMax=<2 5 3>, angularLimitsMin=<-7 -5 -3>, angularTarget=<0 0 0>}
      localJointOffset=<0 0 -4.8>
     }]
   }
   n10 = Anim Dynamics - Bone: dyn_back_bow_string02_r(ComponentPose=@n9, Alpha=@n4, ExternalForce="0,0,0") {
    angularDampingOverride=0.99
    bGravityOverrideInSimSpace=true
    bOverrideAngularDamping=true
    gravityScale=0.35
    numSolverIterationsPostUpdate=2
    numSolverIterationsPreUpdate=8
    simulationSpace=RootRelative
    alphaBoolBlend {blendOption=Linear}
    boundBone {boneName=dyn_back_bow_string02_r}
    componentAppliedLinearAccClamp=<3000 3000 3000>
    physicsBodyDefinitions [{
      sphereCollisionRadius=0
      boundBone {boneName=dyn_back_bow_string02_r}
      boxExtents=<5 5 2>
      constraintSetup {angularTargetAxis=AxisZ, twistAxis=AxisZ, angularLimitsMax=<2 5 7>, angularLimitsMin=<-7 -5 -7>, angularTarget=<0 0 0>}
      localJointOffset=<0 0 -5>
     }]
   }
   n11 = "Transform (Modify) Bone - Bone: dyn_hip_bow_l"!(ComponentPose=@n10, bAlphaBoolEnabled:bool=Get bIsCrouching(), Translation="0.5,-2.5,1", Rotation="5,-55,-5", Scale="1,1,1") -> (Pose) {
    alphaInputType=Bool
    rotationMode=BMM_Additive
    rotationSpace=BCS_BoneSpace
    translationMode=BMM_Additive
    translationSpace=BCS_BoneSpace
    alphaBoolBlend {blendInTime=0.1, blendOption=Linear, blendOutTime=0.1}
    boneToModify {boneName=dyn_hip_bow_l}
    rotation=<rot 5 -55 -5>
    translation=<0.5 -2.5 1>
   }
   n12 = Anim Dynamics - Bone: dyn_hip_bow_l(ComponentPose=@n11, Alpha=1, ExternalForce="0,0,0") {
    angularDampingOverride=0.995
    angularSpringConstant=2000
    bAngularSpring=true
    bGravityOverrideInSimSpace=true
    bOverrideAngularDamping=true
    gravityScale=0.35
    numSolverIterationsPostUpdate=2
    numSolverIterationsPreUpdate=8
    simulationSpace=RootRelative
    alphaBoolBlend {blendOption=Linear}
    boundBone {boneName=dyn_hip_bow_l}
    componentAppliedLinearAccClamp=<3000 3000 3000>
    componentLinearAccScale=<0.1 0.1 0.1>
    componentLinearVelScale=<0.1 0.1 0.1>
    physicsBodyDefinitions [{
      sphereCollisionRadius=0
      boundBone {boneName=dyn_hip_bow_l}
      boxExtents=<10 35 35>
      constraintSetup {angularTargetAxis=AxisZ, twistAxis=AxisZ, angularLimitsMax=<5 0.5 3>, angularLimitsMin=<-5 -0.5 -3>, angularTarget=<0 0 0>}
      localJointOffset=<0 0 -2>
     }]
   }
   n13 = Anim Dynamics - Bone: dyn_hip_bow_string01_r(ComponentPose=@n12, Alpha=1, ExternalForce="0,0,0") {
    angularDampingOverride=0.99
    bGravityOverrideInSimSpace=true
    bOverrideAngularDamping=true
    gravityScale=0.35
    numSolverIterationsPostUpdate=2
    numSolverIterationsPreUpdate=8
    simulationSpace=RootRelative
    alphaBoolBlend {blendOption=Linear}
    boundBone {boneName=dyn_hip_bow_string01_r}
    componentAppliedLinearAccClamp=<3000 3000 3000>
    componentLinearAccScale=<0.2 0.2 0.35>
    componentLinearVelScale=<0.2 0.2 0.35>
    physicsBodyDefinitions [{
      sphereCollisionRadius=0
      boundBone {boneName=dyn_hip_bow_string01_r}
      boxExtents=<7.5 25 25>
      constraintSetup {angularTargetAxis=AxisZ, twistAxis=AxisZ, angularLimitsMax=<15 4 3>, angularLimitsMin=<-15 -20 -3>, angularTarget=<0 0 0>}
      localJointOffset=<0 0 -5>
     }]
    relativeSpaceBone {boneName=dyn_hip_bow_l}
   }
   n14 = Anim Dynamics - Bone: dyn_hip_bow_string02_r(ComponentPose=@n13, Alpha=@n4, ExternalForce="0,0,0") {
    angularDampingOverride=0.985
    bGravityOverrideInSimSpace=true
    bOverrideAngularDamping=true
    gravityScale=0.35
    numSolverIterationsPostUpdate=2
    numSolverIterationsPreUpdate=8
    simulationSpace=RootRelative
    alphaBoolBlend {blendOption=Linear}
    boundBone {boneName=dyn_hip_bow_string02_r}
    componentAppliedLinearAccClamp=<3000 3000 3000>
    physicsBodyDefinitions [{
      sphereCollisionRadius=0
      boundBone {boneName=dyn_hip_bow_string02_r}
      boxExtents=<5 5 2>
      constraintSetup {angularTargetAxis=AxisZ, twistAxis=AxisZ, angularLimitsMax=<7 0 5>, angularLimitsMin=<-7 -2 -5>, angularTarget=<0 0 0>}
      localJointOffset=<0 0 -4.8>
     }]
   }
   n15 = Anim Dynamics - Bone: dyn_hip_bow_string01_l(ComponentPose=@n14, Alpha=1, ExternalForce="0,0,0") {
    angularDampingOverride=0.987
    bGravityOverrideInSimSpace=true
    bOverrideAngularDamping=true
    gravityScale=0.35
    numSolverIterationsPostUpdate=2
    numSolverIterationsPreUpdate=8
    simulationSpace=RootRelative
    alphaBoolBlend {blendOption=Linear}
    boundBone {boneName=dyn_hip_bow_string01_l}
    componentAppliedLinearAccClamp=<3000 3000 3000>
    componentLinearAccScale=<0.2 0.2 0.35>
    componentLinearVelScale=<0.2 0.2 0.35>
    physicsBodyDefinitions [{
      sphereCollisionRadius=0
      boundBone {boneName=dyn_hip_bow_string01_l}
      boxExtents=<7.5 25 25>
      constraintSetup {angularTargetAxis=AxisZ, twistAxis=AxisZ, angularLimitsMax=<15 4 3>, angularLimitsMin=<-15 -20 -3>, angularTarget=<0 0 0>}
      localJointOffset=<0 0 -4.8>
     }]
    relativeSpaceBone {boneName=dyn_hip_bow_l}
   }
   Anim Dynamics - Bone: dyn_hip_bow_string02_l(ComponentPose=@n15, Alpha=@n4, ExternalForce="0,0,0") {
    angularDampingOverride=0.99
    bGravityOverrideInSimSpace=true
    bOverrideAngularDamping=true
    gravityScale=0.35
    numSolverIterationsPostUpdate=2
    numSolverIterationsPreUpdate=8
    simulationSpace=RootRelative
    alphaBoolBlend {blendOption=Linear}
    boundBone {boneName=dyn_hip_bow_string02_l}
    componentAppliedLinearAccClamp=<3000 3000 3000>
    physicsBodyDefinitions [{
      sphereCollisionRadius=0
      boundBone {boneName=dyn_hip_bow_string02_l}
      boxExtents=<5 5 2>
      constraintSetup {angularTargetAxis=AxisZ, twistAxis=AxisZ, angularLimitsMax=<7 0 7>, angularLimitsMin=<-7 -2 -7>, angularTarget=<0 0 0>}
      localJointOffset=<0 0 -5>
     }]
   }
   n16 = Get Skydiving Vector Force() -> (Skydiving Vector Force:Vector)
   Anim Dynamics - Bone: dyn_skirt_fr_01_l(Alpha=1, ExternalForce=@n16) {
    angularDampingOverride=0.98
    bGravityOverrideInSimSpace=true
    bOverrideAngularDamping=true
    bUseGravityOverride=true
    numSolverIterationsPostUpdate=2
    numSolverIterationsPreUpdate=8
    simulationSpace=BoneRelative
    alphaBoolBlend {blendOption=Linear}
    boundBone {boneName=dyn_skirt_fr_01_l}
    componentAppliedLinearAccClamp=<3000 3000 3000>
    componentLinearAccScale=<0.1 0.1 0.1>
    componentLinearVelScale=<0.1 0.1 0.1>
    gravityOverride {z=-980}
    physicsBodyDefinitions [{
      sphereCollisionRadius=0
      boundBone {boneName=dyn_skirt_fr_01_l}
      boxExtents=<10 20 20>
      constraintSetup {angularLimitsMax=<0 5 0>, angularLimitsMin=<0 -5 -30>, angularTarget=<0 0 0>}
      localJointOffset=<-0.32 0 0>
     }]
    relativeSpaceBone {boneName=thigh_l}
   }
   n17 = Anim Dynamics - Bone: dyn_arm_bow_l(Alpha=1, ExternalForce="0,0,0") {
    angularDampingOverride=0.99
    angularSpringConstant=1500
    bAngularSpring=true
    bGravityOverrideInSimSpace=true
    bOverrideAngularDamping=true
    gravityScale=0.35
    numSolverIterationsPostUpdate=2
    numSolverIterationsPreUpdate=8
    simulationSpace=RootRelative
    alphaBoolBlend {blendOption=Linear}
    boundBone {boneName=dyn_arm_bow_l}
    componentAppliedLinearAccClamp=<3000 3000 3000>
    componentLinearAccScale=<0.3 0.3 0.3>
    componentLinearVelScale=<0.3 0.3 0.3>
    physicsBodyDefinitions [{
      sphereCollisionRadius=0
      boundBone {boneName=dyn_arm_bow_l}
      boxExtents=<20 20 10>
      constraintSetup {angularTargetAxis=AxisZ, twistAxis=AxisZ, angularLimitsMax=<-5 3 5>, angularLimitsMin=<-5 -3 -5>, angularTarget=<0 0 0>}
      localJointOffset=<0 0 -2>
     }]
   }
   Anim Dynamics - Bone: dyn_arm_bow_r(ComponentPose=@n17, Alpha=1, ExternalForce="0,0,0") {
    angularDampingOverride=0.99
    angularSpringConstant=1500
    bAngularSpring=true
    bGravityOverrideInSimSpace=true
    bOverrideAngularDamping=true
    gravityScale=0.35
    numSolverIterationsPostUpdate=2
    numSolverIterationsPreUpdate=8
    simulationSpace=RootRelative
    alphaBoolBlend {blendOption=Linear}
    boundBone {boneName=dyn_arm_bow_r}
    componentAppliedLinearAccClamp=<3000 3000 3000>
    componentLinearAccScale=<0.3 0.3 0.3>
    componentLinearVelScale=<0.3 0.3 0.3>
    physicsBodyDefinitions [{
      sphereCollisionRadius=0
      boundBone {boneName=dyn_arm_bow_r}
      boxExtents=<20 20 10>
      constraintSetup {angularTargetAxis=AxisZ, twistAxis=AxisZ, angularLimitsMax=<5 0 3>, angularLimitsMin=<-5 0 -3>, angularTarget=<0 0 0>}
      localJointOffset=<0 0 -2>
     }]
   }
   n18 = Anim Dynamics - Bone: dyn_chest_bow(Alpha=1, ExternalForce="0,0,0") {
    angularDampingOverride=0.99
    angularSpringConstant=1500
    bAngularSpring=true
    bGravityOverrideInSimSpace=true
    bOverrideAngularDamping=true
    gravityScale=0.35
    numSolverIterationsPostUpdate=2
    numSolverIterationsPreUpdate=8
    simulationSpace=RootRelative
    alphaBoolBlend {blendOption=Linear}
    boundBone {boneName=dyn_chest_bow}
    componentAppliedLinearAccClamp=<3000 3000 3000>
    componentLinearAccScale=<0.5 0.5 0.5>
    componentLinearVelScale=<0.5 0.5 0.5>
    physicsBodyDefinitions [{
      sphereCollisionRadius=0
      boundBone {boneName=dyn_chest_bow}
      boxExtents=<20 20 5>
      constraintSetup {angularTargetAxis=AxisZ, twistAxis=AxisZ, angularLimitsMax=<5 2 3>, angularLimitsMin=<-5 -2 -3>, angularTarget=<0 0 0>}
      localJointOffset=<0 0 -2>
     }]
   }
   n19 = Anim Dynamics - Bone: dyn_chest_bow_string01_l(ComponentPose=@n18, Alpha=1, ExternalForce="0,0,0") {
    angularDampingOverride=0.98
    bGravityOverrideInSimSpace=true
    bOverrideAngularDamping=true
    gravityScale=0.35
    numSolverIterationsPostUpdate=2
    numSolverIterationsPreUpdate=8
    simulationSpace=RootRelative
    alphaBoolBlend {blendOption=Linear}
    boundBone {boneName=dyn_chest_bow_string01_l}
    componentAppliedLinearAccClamp=<3000 3000 3000>
    componentLinearAccScale=<0.5 0.5 0.5>
    componentLinearVelScale=<0.5 0.5 0.5>
    physicsBodyDefinitions [{
      sphereCollisionRadius=0
      boundBone {boneName=dyn_chest_bow_string01_l}
      boxExtents=<15 15 5>
      constraintSetup {angularTargetAxis=AxisZ, twistAxis=AxisZ, angularLimitsMax=<20 5 3>, angularLimitsMin=<-10 -5 -3>, angularTarget=<0 0 0>}
      localJointOffset=<0 0 -1>
     }]
   }
   Anim Dynamics - Bone: dyn_chest_bow_string01_r(ComponentPose=@n19, Alpha=1, ExternalForce="0,0,0") {
    angularDampingOverride=0.95
    bGravityOverrideInSimSpace=true
    bOverrideAngularDamping=true
    gravityScale=0.35
    numSolverIterationsPostUpdate=2
    numSolverIterationsPreUpdate=8
    simulationSpace=RootRelative
    alphaBoolBlend {blendOption=Linear}
    boundBone {boneName=dyn_chest_bow_string01_r}
    componentAppliedLinearAccClamp=<3000 3000 3000>
    componentLinearAccScale=<0.5 0.5 0.5>
    componentLinearVelScale=<0.5 0.5 0.5>
    physicsBodyDefinitions [{
      sphereCollisionRadius=0
      boundBone {boneName=dyn_chest_bow_string01_r}
      boxExtents=<5 5 2>
      constraintSetup {angularTargetAxis=AxisZ, twistAxis=AxisZ, angularLimitsMax=<20 5 3>, angularLimitsMin=<-10 -5 -3>, angularTarget=<0 0 0>}
      localJointOffset=<0 0 -1>
     }]
   }
   n20 = Get bIsCrouching()
   n21 = "Transform (Modify) Bone - Bone: dyn_skirt_bk_A_01_r"!(ComponentPose, bAlphaBoolEnabled:bool=@n20, Translation="0,0,0", Rotation="0,0,0", Scale="1,1,1") -> (Pose) {
    alphaInputType=Bool
    rotationMode=BMM_Additive
    rotationSpace=BCS_BoneSpace
    translationMode=BMM_Additive
    translationSpace=BCS_BoneSpace
    alphaBoolBlend {blendOption=Linear}
    boneToModify {boneName=dyn_skirt_bk_A_01_r}
   }
   n22 = "Transform (Modify) Bone - Bone: dyn_skirt_frdiag_01_l"!(ComponentPose=@n21, bAlphaBoolEnabled:bool=@n20, Translation="0.000,6.500,5", Rotation="8,10,0.000", Scale="1,1,1") -> (Pose) {
    alphaInputType=Bool
    rotationMode=BMM_Additive
    rotationSpace=BCS_BoneSpace
    translationMode=BMM_Additive
    translationSpace=BCS_BoneSpace
    alphaBoolBlend {blendOption=Linear}
    boneToModify {boneName=dyn_skirt_frdiag_01_l}
    rotation {pitch=8, yaw=10}
    translation {y=6.5, z=5}
   }
   n23 = "Transform (Modify) Bone - Bone: dyn_skirt_frdiag_A_01_l"!(ComponentPose=@n22, bAlphaBoolEnabled:bool=@n20, Translation="-5,-7,5", Rotation="14.000,-3,0.000", Scale="1,1,1") -> (Pose) {
    alphaInputType=Bool
    rotationMode=BMM_Additive
    rotationSpace=BCS_BoneSpace
    translationMode=BMM_Additive
    translationSpace=BCS_BoneSpace
    alphaBoolBlend {blendOption=Linear}
    boneToModify {boneName=dyn_skirt_frdiag_A_01_l}
    rotation {pitch=14, yaw=-3}
    translation=<-5 -7 5>
   }
   n24 = "Transform (Modify) Bone - Bone: dyn_skirt_fr_01_r"!(ComponentPose=@n23, bAlphaBoolEnabled:bool=@n20, Translation="0,0,0", Rotation="0,0,0", Scale="1,1,1") -> (Pose) {
    alphaInputType=Bool
    rotationMode=BMM_Additive
    rotationSpace=BCS_BoneSpace
    translationMode=BMM_Additive
    translationSpace=BCS_BoneSpace
    alphaBoolBlend {blendOption=Linear}
    boneToModify {boneName=dyn_skirt_fr_01_r}
   }
   n25 = "Transform (Modify) Bone - Bone: dyn_skirt_frdiag_A_01_r"!(ComponentPose=@n24, bAlphaBoolEnabled:bool=@n20, Translation="0,0,0", Rotation="0,0,0", Scale="1,1,1") -> (Pose) {
    alphaInputType=Bool
    rotationMode=BMM_Additive
    rotationSpace=BCS_BoneSpace
    translationMode=BMM_Additive
    translationSpace=BCS_BoneSpace
    alphaBoolBlend {blendOption=Linear}
    boneToModify {boneName=dyn_skirt_frdiag_A_01_r}
   }
   n26 = "Transform (Modify) Bone - Bone: dyn_skirt_fr_01"!(ComponentPose=@n25, bAlphaBoolEnabled:bool=@n20, Translation="0,0,0", Rotation="0,0,0", Scale="1,1,1") -> (Pose) {
    alphaInputType=Bool
    rotationMode=BMM_Additive
    rotationSpace=BCS_BoneSpace
    translationMode=BMM_Additive
    translationSpace=BCS_BoneSpace
    alphaBoolBlend {blendOption=Linear}
    boneToModify {boneName=dyn_skirt_fr_01}
   }
   n27 = Anim Dynamics - Bone: dyn_skirt_fr_02_l(ComponentPose=@n26, Alpha=1, ExternalForce=@n16) {
    angularDampingOverride=0.98
    bGravityOverrideInSimSpace=true
    bOverrideAngularDamping=true
    bUseGravityOverride=true
    numSolverIterationsPostUpdate=2
    numSolverIterationsPreUpdate=8
    simulationSpace=BoneRelative
    alphaBoolBlend {blendOption=Linear}
    boundBone {boneName=dyn_skirt_fr_02_l}
    componentAppliedLinearAccClamp=<3000 3000 3000>
    componentLinearAccScale=<0.1 0.1 0.1>
    componentLinearVelScale=<0.1 0.1 0.1>
    gravityOverride {z=-980}
    physicsBodyDefinitions [{
      sphereCollisionRadius=0
      boundBone {boneName=dyn_skirt_fr_02_l}
      boxExtents=<10 20 20>
      constraintSetup {angularLimitsMax=<0 2 3>, angularLimitsMin=<0 -2 -30>, angularTarget=<0 0 0>}
      localJointOffset=<-0.32 0 0>
     }]
    relativeSpaceBone {boneName=thigh_l}
   }
   n28 = Anim Dynamics - Bone: dyn_skirt_fr_03_l(ComponentPose=@n27, Alpha=1, ExternalForce=@n16) {
    angularDampingOverride=0.98
    bGravityOverrideInSimSpace=true
    bOverrideAngularDamping=true
    bUseGravityOverride=true
    numSolverIterationsPostUpdate=2
    numSolverIterationsPreUpdate=8
    simulationSpace=BoneRelative
    alphaBoolBlend {blendOption=Linear}
    boundBone {boneName=dyn_skirt_fr_03_l}
    componentAppliedLinearAccClamp=<3000 3000 3000>
    componentLinearAccScale=<0.1 0.1 0.1>
    componentLinearVelScale=<0.1 0.1 0.1>
    gravityOverride {z=-980}
    physicsBodyDefinitions [{
      sphereCollisionRadius=0
      boundBone {boneName=dyn_skirt_fr_03_l}
      boxExtents=<10 20 20>
      constraintSetup {angularLimitsMax=<0 2 10>, angularLimitsMin=<0 -2 -30>, angularTarget=<0 0 0>}
      localJointOffset=<-0.32 0 0>
     }]
    relativeSpaceBone {boneName=thigh_l}
   }
   n29 = Anim Dynamics - Bone: dyn_skirt_fr_04_l(ComponentPose=@n28, Alpha=1, ExternalForce=@n16) {
    angularDampingOverride=0.98
    bGravityOverrideInSimSpace=true
    bOverrideAngularDamping=true
    bUseGravityOverride=true
    numSolverIterationsPostUpdate=2
    numSolverIterationsPreUpdate=8
    simulationSpace=BoneRelative
    alphaBoolBlend {blendOption=Linear}
    boundBone {boneName=dyn_skirt_fr_04_l}
    componentAppliedLinearAccClamp=<3000 3000 3000>
    componentLinearAccScale=<0.1 0.1 0.1>
    componentLinearVelScale=<0.1 0.1 0.1>
    gravityOverride {z=-980}
    physicsBodyDefinitions [{
      sphereCollisionRadius=0
      boundBone {boneName=dyn_skirt_fr_04_l}
      boxExtents=<10 20 20>
      constraintSetup {angularLimitsMax=<2 2 15>, angularLimitsMin=<-2 -2 -30>, angularTarget=<0 0 0>}
      localJointOffset=<-0.32 0 0>
     }]
    relativeSpaceBone {boneName=thigh_l}
   }
   n30 = Anim Dynamics - Bone: dyn_skirt_fr_05_l(ComponentPose=@n29, Alpha=1, ExternalForce=@n16) {
    angularDampingOverride=0.98
    bGravityOverrideInSimSpace=true
    bOverrideAngularDamping=true
    bUseGravityOverride=true
    numSolverIterationsPostUpdate=2
    numSolverIterationsPreUpdate=8
    simulationSpace=BoneRelative
    alphaBoolBlend {blendOption=Linear}
    boundBone {boneName=dyn_skirt_fr_05_l}
    componentAppliedLinearAccClamp=<3000 3000 3000>
    componentLinearAccScale=<0.1 0.1 0.1>
    componentLinearVelScale=<0.1 0.1 0.1>
    gravityOverride {z=-980}
    physicsBodyDefinitions [{
      sphereCollisionRadius=0
      boundBone {boneName=dyn_skirt_fr_05_l}
      boxExtents=<10 20 20>
      constraintSetup {angularLimitsMax=<2 2 15>, angularLimitsMin=<-2 -2 -30>, angularTarget=<0 0 0>}
      localJointOffset=<-0.32 0 0>
     }]
    relativeSpaceBone {boneName=thigh_l}
   }
   n31 = Anim Dynamics - Bone: dyn_skirt_fr_01_r(ComponentPose=@n30, Alpha=1, ExternalForce=@n16) {
    angularDampingOverride=0.98
    bGravityOverrideInSimSpace=true
    bOverrideAngularDamping=true
    bUseGravityOverride=true
    numSolverIterationsPostUpdate=2
    numSolverIterationsPreUpdate=8
    simulationSpace=BoneRelative
    alphaBoolBlend {blendOption=Linear}
    boundBone {boneName=dyn_skirt_fr_01_r}
    componentAppliedLinearAccClamp=<3000 3000 3000>
    componentLinearAccScale=<0.1 0.1 0.1>
    componentLinearVelScale=<0.1 0.1 0.1>
    gravityOverride {z=-980}
    physicsBodyDefinitions [{
      sphereCollisionRadius=0
      boundBone {boneName=dyn_skirt_fr_01_r}
      boxExtents=<10 20 20>
      constraintSetup {angularLimitsMax=<0 5 0>, angularLimitsMin=<0 -5 -30>, angularTarget=<0 0 0>}
      localJointOffset=<3 -0.1 0>
     }]
    relativeSpaceBone {boneName=thigh_twist_01_r}
   }
   n32 = Anim Dynamics - Bone: dyn_skirt_fr_02_r(ComponentPose=@n31, Alpha=1, ExternalForce=@n16) {
    angularDampingOverride=0.98
    bGravityOverrideInSimSpace=true
    bOverrideAngularDamping=true
    bUseGravityOverride=true
    numSolverIterationsPostUpdate=2
    numSolverIterationsPreUpdate=8
    simulationSpace=BoneRelative
    alphaBoolBlend {blendOption=Linear}
    boundBone {boneName=dyn_skirt_fr_02_r}
    componentAppliedLinearAccClamp=<3000 3000 3000>
    componentLinearAccScale=<0.1 0.1 0.1>
    componentLinearVelScale=<0.1 0.1 0.1>
    gravityOverride {z=-980}
    physicsBodyDefinitions [{
      sphereCollisionRadius=1
      boundBone {boneName=dyn_skirt_fr_02_r}
      boxExtents=<10 20 20>
      constraintSetup {angularTargetAxis=AxisZ, twistAxis=AxisZ, angularLimitsMax=<0 2 5>, angularLimitsMin=<0 -2 -30>, angularTarget=<0 0 0>}
      localJointOffset=<3 -0.1 0>
     }]
    relativeSpaceBone {boneName=thigh_twist_01_r}
    sphericalLimits [{limitRadius=8, drivingBone {boneName=thigh_r}, sphereLocalOffset=<8 0 0>}]
   }
   n33 = Anim Dynamics - Bone: dyn_skirt_fr_03_r(ComponentPose=@n32, Alpha=1, ExternalForce=@n16) {
    angularDampingOverride=0.98
    bGravityOverrideInSimSpace=true
    bOverrideAngularDamping=true
    bUseGravityOverride=true
    numSolverIterationsPostUpdate=2
    numSolverIterationsPreUpdate=8
    simulationSpace=BoneRelative
    alphaBoolBlend {blendOption=Linear}
    boundBone {boneName=dyn_skirt_fr_03_r}
    componentAppliedLinearAccClamp=<3000 3000 3000>
    componentLinearAccScale=<0.1 0.1 0.1>
    componentLinearVelScale=<0.1 0.1 0.1>
    gravityOverride {z=-980}
    physicsBodyDefinitions [{
      sphereCollisionRadius=1
      boundBone {boneName=dyn_skirt_fr_03_r}
      boxExtents=<10 20 20>
      constraintSetup {angularTargetAxis=AxisZ, twistAxis=AxisZ, angularLimitsMax=<2 2 10>, angularLimitsMin=<0 -2 -30>, angularTarget=<0 0 0>}
      localJointOffset=<3 -0.1 0>
     }]
    relativeSpaceBone {boneName=thigh_twist_01_r}
    sphericalLimits [{limitRadius=8, drivingBone {boneName=thigh_r}, sphereLocalOffset=<24 0 0>}]
   }
   n34 = Anim Dynamics - Bone: dyn_skirt_fr_04_r(ComponentPose=@n33, Alpha=1, ExternalForce=@n16) {
    angularDampingOverride=0.98
    bGravityOverrideInSimSpace=true
    bOverrideAngularDamping=true
    bUseGravityOverride=true
    numSolverIterationsPostUpdate=2
    numSolverIterationsPreUpdate=8
    simulationSpace=BoneRelative
    alphaBoolBlend {blendOption=Linear}
    boundBone {boneName=dyn_skirt_fr_04_r}
    componentAppliedLinearAccClamp=<3000 3000 3000>
    componentLinearAccScale=<0.1 0.1 0.1>
    componentLinearVelScale=<0.1 0.1 0.1>
    gravityOverride {z=-980}
    physicsBodyDefinitions [{
      sphereCollisionRadius=1
      boundBone {boneName=dyn_skirt_fr_04_r}
      boxExtents=<10 20 20>
      constraintSetup {angularTargetAxis=AxisZ, twistAxis=AxisZ, angularLimitsMax=<2 2 30>, angularLimitsMin=<-2 -2 -30>, angularTarget=<0 0 0>}
      localJointOffset=<3 -0.1 0>
     }]
    relativeSpaceBone {boneName=thigh_twist_01_r}
    sphericalLimits [{limitRadius=8, drivingBone {boneName=thigh_r}, sphereLocalOffset=<24 0 0>}]
   }
   n35 = Anim Dynamics - Bone: dyn_skirt_fr_05_r(ComponentPose=@n34, Alpha=1, ExternalForce=@n16) {
    angularDampingOverride=0.98
    bGravityOverrideInSimSpace=true
    bOverrideAngularDamping=true
    bUseGravityOverride=true
    numSolverIterationsPostUpdate=2
    numSolverIterationsPreUpdate=8
    simulationSpace=BoneRelative
    alphaBoolBlend {blendOption=Linear}
    boundBone {boneName=dyn_skirt_fr_05_r}
    componentAppliedLinearAccClamp=<3000 3000 3000>
    componentLinearAccScale=<0.1 0.1 0.1>
    componentLinearVelScale=<0.1 0.1 0.1>
    gravityOverride {z=-980}
    physicsBodyDefinitions [{
      sphereCollisionRadius=1
      boundBone {boneName=dyn_skirt_fr_05_r}
      boxExtents=<10 20 20>
      constraintSetup {angularTargetAxis=AxisZ, twistAxis=AxisZ, angularLimitsMax=<2 2 30>, angularLimitsMin=<-2 -2 -30>, angularTarget=<0 0 0>}
      localJointOffset=<3 -0.1 0>
     }]
    relativeSpaceBone {boneName=thigh_twist_01_r}
    sphericalLimits [{limitRadius=8, drivingBone {boneName=thigh_r}, sphereLocalOffset=<24 0 0>}]
   }
   n36 = Anim Dynamics - Bone: dyn_skirt_frdiag_A_01_l(ComponentPose=@n35, Alpha=1, ExternalForce=@n16) {
    angularDampingOverride=0.98
    bGravityOverrideInSimSpace=true
    bOverrideAngularDamping=true
    bUseGravityOverride=true
    numSolverIterationsPostUpdate=2
    numSolverIterationsPreUpdate=8
    simulationSpace=BoneRelative
    alphaBoolBlend {blendOption=Linear}
    boundBone {boneName=dyn_skirt_frdiag_A_01_l}
    componentAppliedLinearAccClamp=<3000 3000 3000>
    componentLinearAccScale=<0.1 0.1 0.1>
    componentLinearVelScale=<0.1 0.1 0.1>
    gravityOverride {z=-980}
    physicsBodyDefinitions [{
      sphereCollisionRadius=0
      boundBone {boneName=dyn_skirt_frdiag_A_01_l}
      boxExtents=<10 20 20>
      constraintSetup {angularLimitsMax=<0 0 5>, angularLimitsMin=<0 0 -5>, angularTarget=<0 0 0>}
      localJointOffset=<3 0 0>
     }]
    relativeSpaceBone {boneName=thigh_l}
   }
   n37 = Anim Dynamics - Bone: dyn_skirt_frdiag_A_02_l(ComponentPose=@n36, Alpha=1, ExternalForce=@n16) {
    angularDampingOverride=0.98
    bGravityOverrideInSimSpace=true
    bOverrideAngularDamping=true
    bUseGravityOverride=true
    numSolverIterationsPostUpdate=2
    numSolverIterationsPreUpdate=8
    simulationSpace=BoneRelative
    alphaBoolBlend {blendOption=Linear}
    boundBone {boneName=dyn_skirt_frdiag_A_02_l}
    componentAppliedLinearAccClamp=<3000 3000 3000>
    componentLinearAccScale=<0.1 0.1 0.1>
    componentLinearVelScale=<0.1 0.1 0.1>
    gravityOverride {z=-980}
    physicsBodyDefinitions [{
      sphereCollisionRadius=1
      boundBone {boneName=dyn_skirt_frdiag_A_02_l}
      boxExtents=<10 20 20>
      constraintSetup {angularTargetAxis=AxisZ, twistAxis=AxisZ, angularLimitsMax=<0 0 5>, angularLimitsMin=<0 0 -5>, angularTarget=<0 0 0>}
      localJointOffset=<3 0 0>
     }]
    relativeSpaceBone {boneName=thigh_l}
    sphericalLimits [{limitRadius=9.1, drivingBone {boneName=thigh_r}, sphereLocalOffset=<8 0 0>}]
   }
   n38 = Anim Dynamics - Bone: dyn_skirt_frdiag_A_03_l(ComponentPose=@n37, Alpha=1, ExternalForce=@n16) {
    angularDampingOverride=0.98
    bGravityOverrideInSimSpace=true
    bOverrideAngularDamping=true
    bUseGravityOverride=true
    numSolverIterationsPostUpdate=2
    numSolverIterationsPreUpdate=8
    simulationSpace=BoneRelative
    alphaBoolBlend {blendOption=Linear}
    boundBone {boneName=dyn_skirt_frdiag_A_03_l}
    componentAppliedLinearAccClamp=<3000 3000 3000>
    componentLinearAccScale=<0.1 0.1 0.1>
    componentLinearVelScale=<0.1 0.1 0.1>
    gravityOverride {z=-980}
    physicsBodyDefinitions [{
      sphereCollisionRadius=1
      boundBone {boneName=dyn_skirt_frdiag_A_03_l}
      boxExtents=<10 20 20>
      constraintSetup {angularTargetAxis=AxisZ, twistAxis=AxisZ, angularLimitsMax=<0 5 5>, angularLimitsMin=<0 0 -5>, angularTarget=<0 0 0>}
      localJointOffset=<3 0 0>
     }]
    relativeSpaceBone {boneName=thigh_l}
    sphericalLimits [{limitRadius=9.2, drivingBone {boneName=thigh_r}, sphereLocalOffset=<24 0 0>}]
   }
   n39 = Anim Dynamics - Bone: dyn_skirt_frdiag_A_04_l(ComponentPose=@n38, Alpha=1, ExternalForce=@n16) {
    angularDampingOverride=0.98
    bGravityOverrideInSimSpace=true
    bOverrideAngularDamping=true
    bUseGravityOverride=true
    numSolverIterationsPostUpdate=2
    numSolverIterationsPreUpdate=8
    simulationSpace=BoneRelative
    alphaBoolBlend {blendOption=Linear}
    boundBone {boneName=dyn_skirt_frdiag_A_04_l}
    componentAppliedLinearAccClamp=<3000 3000 3000>
    componentLinearAccScale=<0.1 0.1 0.1>
    componentLinearVelScale=<0.1 0.1 0.1>
    gravityOverride {z=-980}
    physicsBodyDefinitions [{
      sphereCollisionRadius=1
      boundBone {boneName=dyn_skirt_frdiag_A_04_l}
      boxExtents=<10 20 20>
      constraintSetup {angularTargetAxis=AxisZ, twistAxis=AxisZ, angularLimitsMax=<0 5 5>, angularLimitsMin=<0 0 -5>, angularTarget=<0 0 0>}
      localJointOffset=<3 0 0>
     }]
    relativeSpaceBone {boneName=thigh_l}
    sphericalLimits [{limitRadius=9.2, drivingBone {boneName=thigh_r}, sphereLocalOffset=<24 0 0>}]
   }
   n40 = Anim Dynamics - Bone: dyn_skirt_frdiag_A_05_l(ComponentPose=@n39, Alpha=1, ExternalForce=@n16) {
    angularDampingOverride=0.98
    bGravityOverrideInSimSpace=true
    bOverrideAngularDamping=true
    bUseGravityOverride=true
    numSolverIterationsPostUpdate=2
    numSolverIterationsPreUpdate=8
    simulationSpace=BoneRelative
    alphaBoolBlend {blendOption=Linear}
    boundBone {boneName=dyn_skirt_frdiag_A_05_l}
    componentAppliedLinearAccClamp=<3000 3000 3000>
    componentLinearAccScale=<0.1 0.1 0.1>
    componentLinearVelScale=<0.1 0.1 0.1>
    gravityOverride {z=-980}
    physicsBodyDefinitions [{
      sphereCollisionRadius=1
      boundBone {boneName=dyn_skirt_frdiag_A_05_l}
      boxExtents=<10 20 20>
      constraintSetup {angularTargetAxis=AxisZ, twistAxis=AxisZ, angularLimitsMax=<0 5 5>, angularLimitsMin=<0 -5 -5>, angularTarget=<0 0 0>}
      localJointOffset=<3 0 0>
     }]
    relativeSpaceBone {boneName=thigh_l}
    sphericalLimits [{limitRadius=9.2, drivingBone {boneName=thigh_r}, sphereLocalOffset=<24 0 0>}]
   }
   n41 = "Constraint - dyn_skirt_frdiag_A_01_l (P)"(ComponentPose=@n40, Alpha=1, ConstraintWeights_0=1) {
    alphaBoolBlend {blendOption=Linear}
    boneToModify {boneName=dyn_skirt_frdiag_A_01_l}
    constraintSetup [{targetBone {boneName=thigh_l}}]
    constraintWeights [1]
   }
   n42 = "Constraint - dyn_skirt_frdiag_A_01_r (P)"(ComponentPose=@n41, Alpha=1, ConstraintWeights_0=1) {
    alphaBoolBlend {blendOption=Linear}
    boneToModify {boneName=dyn_skirt_frdiag_A_01_r}
    constraintSetup [{targetBone {boneName=thigh_r}}]
    constraintWeights [1]
   }
   n43 = Anim Dynamics - Bone: dyn_skirt_frdiag_A_01_r(ComponentPose=@n42, Alpha=1, ExternalForce="0,0,0") {
    angularDampingOverride=0.98
    bGravityOverrideInSimSpace=true
    bOverrideAngularDamping=true
    bUseGravityOverride=true
    numSolverIterationsPostUpdate=2
    numSolverIterationsPreUpdate=8
    simulationSpace=BoneRelative
    alphaBoolBlend {blendOption=Linear}
    boundBone {boneName=dyn_skirt_frdiag_A_01_r}
    componentAppliedLinearAccClamp=<3000 3000 3000>
    componentLinearAccScale=<0.1 0.1 0.1>
    componentLinearVelScale=<0.1 0.1 0.1>
    gravityOverride {z=-980}
    physicsBodyDefinitions [{
      sphereCollisionRadius=0
      boundBone {boneName=dyn_skirt_frdiag_A_01_r}
      boxExtents=<10 20 20>
      constraintSetup {angularLimitsMax=<0 0 5>, angularLimitsMin=<0 0 -5>, angularTarget=<0 0 0>}
      localJointOffset=<3 0 0>
     }]
    relativeSpaceBone {boneName=thigh_r}
   }
   n44 = Anim Dynamics - Bone: dyn_skirt_frdiag_A_02_r(ComponentPose=@n43, Alpha=1, ExternalForce="0,0,0") {
    angularDampingOverride=0.98
    bGravityOverrideInSimSpace=true
    bOverrideAngularDamping=true
    bUseGravityOverride=true
    numSolverIterationsPostUpdate=2
    numSolverIterationsPreUpdate=8
    simulationSpace=BoneRelative
    alphaBoolBlend {blendOption=Linear}
    boundBone {boneName=dyn_skirt_frdiag_A_02_r}
    componentAppliedLinearAccClamp=<3000 3000 3000>
    componentLinearAccScale=<0.1 0.1 0.1>
    componentLinearVelScale=<0.1 0.1 0.1>
    gravityOverride {z=-980}
    physicsBodyDefinitions [{
      sphereCollisionRadius=1
      boundBone {boneName=dyn_skirt_frdiag_A_02_r}
      boxExtents=<10 20 20>
      constraintSetup {angularTargetAxis=AxisZ, twistAxis=AxisZ, angularLimitsMax=<0 0 5>, angularLimitsMin=<0 0 -5>, angularTarget=<0 0 0>}
      localJointOffset=<3 0 0>
     }]
    relativeSpaceBone {boneName=thigh_r}
    sphericalLimits [{limitRadius=9.1, drivingBone {boneName=thigh_r}, sphereLocalOffset=<8 0 0>}]
   }
   n45 = Anim Dynamics - Bone: dyn_skirt_frdiag_A_03_r(ComponentPose=@n44, Alpha=1, ExternalForce="0,0,0") {
    angularDampingOverride=0.98
    bGravityOverrideInSimSpace=true
    bOverrideAngularDamping=true
    bUseGravityOverride=true
    numSolverIterationsPostUpdate=2
    numSolverIterationsPreUpdate=8
    simulationSpace=BoneRelative
    alphaBoolBlend {blendOption=Linear}
    boundBone {boneName=dyn_skirt_frdiag_A_03_r}
    componentAppliedLinearAccClamp=<3000 3000 3000>
    componentLinearAccScale=<0.1 0.1 0.1>
    componentLinearVelScale=<0.1 0.1 0.1>
    gravityOverride {z=-980}
    physicsBodyDefinitions [{
      sphereCollisionRadius=1
      boundBone {boneName=dyn_skirt_frdiag_A_03_r}
      boxExtents=<10 20 20>
      constraintSetup {angularTargetAxis=AxisZ, twistAxis=AxisZ, angularLimitsMax=<0 5 5>, angularLimitsMin=<0 0 -5>, angularTarget=<0 0 0>}
      localJointOffset=<3 0 0>
     }]
    relativeSpaceBone {boneName=thigh_r}
    sphericalLimits [{limitRadius=9.2, drivingBone {boneName=thigh_r}, sphereLocalOffset=<24 0 0>}]
   }
   n46 = Anim Dynamics - Bone: dyn_skirt_frdiag_A_04_r(ComponentPose=@n45, Alpha=1, ExternalForce="0,0,0") {
    angularDampingOverride=0.98
    bGravityOverrideInSimSpace=true
    bOverrideAngularDamping=true
    bUseGravityOverride=true
    numSolverIterationsPostUpdate=2
    numSolverIterationsPreUpdate=8
    simulationSpace=BoneRelative
    alphaBoolBlend {blendOption=Linear}
    boundBone {boneName=dyn_skirt_frdiag_A_04_r}
    componentAppliedLinearAccClamp=<3000 3000 3000>
    componentLinearAccScale=<0.1 0.1 0.1>
    componentLinearVelScale=<0.1 0.1 0.1>
    gravityOverride {z=-980}
    physicsBodyDefinitions [{
      sphereCollisionRadius=1
      boundBone {boneName=dyn_skirt_frdiag_A_04_r}
      boxExtents=<10 20 20>
      constraintSetup {angularTargetAxis=AxisZ, twistAxis=AxisZ, angularLimitsMax=<0 5 5>, angularLimitsMin=<0 0 -5>, angularTarget=<0 0 0>}
      localJointOffset=<3 0 0>
     }]
    relativeSpaceBone {boneName=thigh_r}
    sphericalLimits [{limitRadius=9.2, drivingBone {boneName=thigh_r}, sphereLocalOffset=<24 0 0>}]
   }
   n47 = Anim Dynamics - Bone: dyn_skirt_frdiag_A_05_r(ComponentPose=@n46, Alpha=1, ExternalForce="0,0,0") {
    angularDampingOverride=0.98
    bGravityOverrideInSimSpace=true
    bOverrideAngularDamping=true
    bUseGravityOverride=true
    numSolverIterationsPostUpdate=2
    numSolverIterationsPreUpdate=8
    simulationSpace=BoneRelative
    alphaBoolBlend {blendOption=Linear}
    boundBone {boneName=dyn_skirt_frdiag_A_05_r}
    componentAppliedLinearAccClamp=<3000 3000 3000>
    componentLinearAccScale=<0.1 0.1 0.1>
    componentLinearVelScale=<0.1 0.1 0.1>
    gravityOverride {z=-980}
    physicsBodyDefinitions [{
      sphereCollisionRadius=1
      boundBone {boneName=dyn_skirt_frdiag_A_05_r}
      boxExtents=<10 20 20>
      constraintSetup {angularTargetAxis=AxisZ, twistAxis=AxisZ, angularLimitsMax=<0 5 5>, angularLimitsMin=<0 -5 -5>, angularTarget=<0 0 0>}
      localJointOffset=<3 0 0>
     }]
    relativeSpaceBone {boneName=thigh_r}
    sphericalLimits [{limitRadius=9.2, drivingBone {boneName=thigh_r}, sphereLocalOffset=<24 0 0>}]
   }
   n48 = Get SkydiveAlpha() -> (SkydiveAlpha:double)
   n49 = "Constraint - dyn_skirt_fr_01 (PP)"!(ComponentPose=@n47, Alpha=@n48, ConstraintWeights_0=0.5, ConstraintWeights_1:float=0.5) -> (Pose) {
    alphaBoolBlend {blendOption=Linear}
    boneToModify {boneName=dyn_skirt_fr_01}
    constraintSetup [{targetBone {boneName=dyn_skirt_fr_01_l}}, {targetBone {boneName=dyn_skirt_fr_01_r}}]
    constraintWeights [1,1]
   }
   n50 = "Constraint - dyn_skirt_fr_02 (PP)"!(ComponentPose=@n49, Alpha=@n48, ConstraintWeights_0=0.5, ConstraintWeights_1:float=0.5) -> (Pose) {
    alphaBoolBlend {blendOption=Linear}
    boneToModify {boneName=dyn_skirt_fr_02}
    constraintSetup [{targetBone {boneName=dyn_skirt_fr_02_l}}, {targetBone {boneName=dyn_skirt_fr_02_r}}]
    constraintWeights [1,1]
   }
   n51 = "Constraint - dyn_skirt_fr_03 (PP)"!(ComponentPose=@n50, Alpha=@n48, ConstraintWeights_0=0.5, ConstraintWeights_1:float=0.5) -> (Pose) {
    alphaBoolBlend {blendOption=Linear}
    boneToModify {boneName=dyn_skirt_fr_03}
    constraintSetup [{targetBone {boneName=dyn_skirt_fr_03_l}}, {targetBone {boneName=dyn_skirt_fr_03_r}}]
    constraintWeights [1,1]
   }
   "Constraint - dyn_skirt_fr_04 (PP)"!(ComponentPose=@n51, Alpha=@n48, ConstraintWeights_0=0.5, ConstraintWeights_1:float=0.5) -> (Pose) {
    alphaBoolBlend {blendOption=Linear}
    boneToModify {boneName=dyn_skirt_fr_04}
    constraintSetup [{targetBone {boneName=dyn_skirt_fr_04_l}}, {targetBone {boneName=dyn_skirt_fr_04_r}}]
    constraintWeights [1,1]
   }
   RigidBody(ExternalForce=Get Skydiving Vector Force()) {
    bOverrideWorldGravity=true
    overridePhysicsAsset=&/F_MED_DeLasMuertos_Physics_Crouch
    simulationSpace=BaseBoneSpace
    alphaBoolBlend {blendOption=Linear}
    baseBoneRef {boneName=root}
    componentAppliedLinearAccClamp=<50 50 50>
    componentLinearAccScale=<0.2 0.2 0.5>
    componentLinearVelScale=<0.2 0.2 0.5>
    overrideWorldGravity {z=-980}
   }
   n52 = Get Skydiving Vector Force()
   n53 = Anim Dynamics - Bone: dyn_chest_bow_string01_l(ComponentPose=Local To Component(LocalPose=Copy Pose From Mesh() {bUseAttachedParent=true}), Alpha=1, ExternalForce="0,0,0") {
    angularDampingOverride=0.98
    bGravityOverrideInSimSpace=true
    bOverrideAngularDamping=true
    gravityScale=0.35
    numSolverIterationsPostUpdate=2
    numSolverIterationsPreUpdate=8
    simulationSpace=RootRelative
    alphaBoolBlend {blendOption=Linear}
    boundBone {boneName=dyn_chest_bow_string01_l}
    componentAppliedLinearAccClamp=<3000 3000 3000>
    componentLinearAccScale=<0.5 0.5 0.5>
    componentLinearVelScale=<0.5 0.5 0.5>
    physicsBodyDefinitions [{
      sphereCollisionRadius=0
      boundBone {boneName=dyn_chest_bow_string01_l}
      boxExtents=<15 15 5>
      constraintSetup {angularTargetAxis=AxisZ, twistAxis=AxisZ, angularLimitsMax=<20 5 3>, angularLimitsMin=<-10 -5 -3>, angularTarget=<0 0 0>}
      localJointOffset=<0 0 -1>
     }]
   }
   n54 = "Constraint - dyn_chest_bow_string01_r (P)"(ComponentPose=@n53, Alpha=0.65, ConstraintWeights_0=1) {
    alphaBoolBlend {blendOption=Linear}
    boneToModify {boneName=dyn_chest_bow_string01_r}
    constraintSetup [{targetBone {boneName=dyn_chest_bow_string01_l}}]
    constraintWeights [1]
   }
   n55 = Anim Dynamics - Bone: dyn_collar_flower(ComponentPose=@n54, Alpha=1, ExternalForce="0,0,0") {
    angularDampingOverride=0.99
    angularSpringConstant=1500
    bAngularSpring=true
    bGravityOverrideInSimSpace=true
    bOverrideAngularDamping=true
    gravityScale=0.35
    numSolverIterationsPostUpdate=2
    numSolverIterationsPreUpdate=8
    simulationSpace=RootRelative
    alphaBoolBlend {blendOption=Linear}
    boundBone {boneName=dyn_collar_flower}
    componentAppliedLinearAccClamp=<3000 3000 3000>
    componentLinearAccScale=<0.2 0.2 0.2>
    componentLinearVelScale=<0.2 0.2 0.2>
    physicsBodyDefinitions [{
      sphereCollisionRadius=0
      boundBone {boneName=dyn_collar_flower}
      boxExtents=<20 20 10>
      constraintSetup {angularTargetAxis=AxisZ, twistAxis=AxisZ, angularLimitsMax=<5 3 5>, angularLimitsMin=<-5 -3 -5>, angularTarget=<0 0 0>}
      localJointOffset=<0 0 -2>
     }]
   }
   n56 = "Transform (Modify) Bone - Bone: dyn_back_bow"!(
    ComponentPose=@n55
    bAlphaBoolEnabled:bool=Get isCrouchOrSprint() -> (isCrouchOrSprint:bool)
    Translation="0,0,0"
    Rotation="0,0,15"
    Scale="1,1,1"
   ) -> (Pose) {
    alphaInputType=Bool
    rotationMode=BMM_Additive
    rotationSpace=BCS_BoneSpace
    translationMode=BMM_Additive
    translationSpace=BCS_BoneSpace
    alphaBoolBlend {blendInTime=0.25, blendOption=Linear, blendOutTime=0.25}
    boneToModify {boneName=dyn_back_bow}
    rotation {roll=15}
   }
   n57 = Anim Dynamics - Bone: dyn_back_bow_string01_l(ComponentPose=@n56, Alpha=0.7, ExternalForce="0,0,0") {
    angularDampingOverride=0.99
    bGravityOverrideInSimSpace=true
    bOverrideAngularDamping=true
    gravityScale=0.35
    numSolverIterationsPostUpdate=2
    numSolverIterationsPreUpdate=8
    simulationSpace=RootRelative
    alphaBoolBlend {blendOption=Linear}
    boundBone {boneName=dyn_back_bow_string01_l}
    componentAppliedLinearAccClamp=<3000 3000 3000>
    componentLinearAccScale=<0.25 0.2 0.35>
    componentLinearVelScale=<0.25 0.25 0.35>
    physicsBodyDefinitions [{
      sphereCollisionRadius=0
      boundBone {boneName=dyn_back_bow_string01_l}
      boxExtents=<15 10 15>
      constraintSetup {angularTargetAxis=AxisZ, twistAxis=AxisZ, angularLimitsMax=<0 25 3>, angularLimitsMin=<-60 -30 -3>, angularTarget=<0 0 0>}
      localJointOffset=<0 0 -2>
     }]
    relativeSpaceBone {boneName=dyn_back_bow}
   }
   n58 = Anim Dynamics - Bone: dyn_back_bow_string01_r(ComponentPose=@n57, Alpha=0.6, ExternalForce="0,0,0") {
    angularDampingOverride=0.99
    bGravityOverrideInSimSpace=true
    bOverrideAngularDamping=true
    gravityScale=0.35
    numSolverIterationsPostUpdate=2
    numSolverIterationsPreUpdate=8
    simulationSpace=RootRelative
    alphaBoolBlend {blendOption=Linear}
    boundBone {boneName=dyn_back_bow_string01_r}
    componentAppliedLinearAccClamp=<3000 3000 3000>
    componentLinearAccScale=<0.2 0.25 0.3>
    componentLinearVelScale=<0.25 0.25 0.35>
    physicsBodyDefinitions [{
      sphereCollisionRadius=0
      boundBone {boneName=dyn_back_bow_string01_r}
      boxExtents=<18 12 18>
      constraintSetup {angularTargetAxis=AxisZ, twistAxis=AxisZ, angularLimitsMax=<0 30 3>, angularLimitsMin=<-55 -25 -3>, angularTarget=<0 0 0>}
      localJointOffset=<0 0 -3>
     }]
    relativeSpaceBone {boneName=dyn_back_bow}
   }
   n59 = "Constraint - dyn_skirt_fr_01_l (R)"(ComponentPose=@n58, Alpha=1, ConstraintWeights_0=1) {
    alphaBoolBlend {blendOption=Linear}
    boneToModify {boneName=dyn_skirt_fr_01_l}
    constraintSetup [{transformType=Rotation, targetBone {boneName=thigh_l}}]
    constraintWeights [1]
   }
   n60 = "Constraint - dyn_skirt_fr_01_r (R)"(ComponentPose=@n59, Alpha=1, ConstraintWeights_0=1) {
    alphaBoolBlend {blendOption=Linear}
    boneToModify {boneName=dyn_skirt_fr_01_r}
    constraintSetup [{transformType=Rotation, targetBone {boneName=thigh_r}}]
    constraintWeights [1]
   }
   n61 = "Constraint - dyn_skirt_frdiag_A_01_l (R)"(ComponentPose=@n60, Alpha=1, ConstraintWeights_0=1) {
    alphaBoolBlend {blendOption=Linear}
    boneToModify {boneName=dyn_skirt_frdiag_A_01_l}
    constraintSetup [{transformType=Rotation, targetBone {boneName=thigh_l}}]
    constraintWeights [1]
   }
   n62 = "Constraint - dyn_skirt_frdiag_A_01_r (R)"(ComponentPose=@n61, Alpha=1, ConstraintWeights_0=1) {
    alphaBoolBlend {blendOption=Linear}
    boneToModify {boneName=dyn_skirt_frdiag_A_01_r}
    constraintSetup [{transformType=Rotation, targetBone {boneName=thigh_r}}]
    constraintWeights [1]
   }
   n63 = "Constraint - dyn_skirt_frdiag_01_l (RR)"!(ComponentPose=@n62, Alpha=1, ConstraintWeights_0=0.5, ConstraintWeights_1:float=0.5) -> (Pose) {
    alphaBoolBlend {blendOption=Linear}
    boneToModify {boneName=dyn_skirt_frdiag_01_l}
    constraintSetup [{transformType=Rotation, targetBone {boneName=dyn_skirt_frdiag_A_01_l}}, {transformType=Rotation, targetBone {boneName=dyn_skirt_out_01_l}}]
    constraintWeights [1,1]
   }
   n64 = "Constraint - dyn_skirt_frdiag_01_r (RR)"!(ComponentPose=@n63, Alpha=1, ConstraintWeights_0=0.5, ConstraintWeights_1:float=0.5) -> (Pose) {
    alphaBoolBlend {blendOption=Linear}
    boneToModify {boneName=dyn_skirt_frdiag_01_r}
    constraintSetup [{transformType=Rotation, targetBone {boneName=dyn_skirt_frdiag_A_01_r}}, {transformType=Rotation, targetBone {boneName=dyn_skirt_out_01_r}}]
    constraintWeights [1,1]
   }
   n65 = "dyn_skirt_frdiag_01_l.rotateY = thigh_l.rotateY * -0.1 - Bone Driven Controller"!(ComponentPose:Component Space Pose Link Structure=@n64, bAlphaBoolEnabled:bool=True) -> (Pose:Component Space Pose Link Structure) {
    is=title AnimGraphNode_BoneDrivenController
    alphaInputType=Bool
    bAffectTargetRotationY=true
    multiplier=-0.1
    sourceComponent=RotationY
    alphaBoolBlend {blendOption=Linear}
    sourceBone {boneName=thigh_l}
    targetBone {boneName=dyn_skirt_frdiag_01_l}
   }
   n66 = "dyn_skirt_frdiag_A_01_l.rotateY = thigh_l.rotateZ * -0.05 - Bone Driven Controller"!(ComponentPose:Component Space Pose Link Structure=@n65, bAlphaBoolEnabled:bool=True) -> (Pose:Component Space Pose Link Structure) {
    is=title AnimGraphNode_BoneDrivenController
    alphaInputType=Bool
    bAffectTargetRotationY=true
    multiplier=-0.05
    sourceComponent=RotationZ
    alphaBoolBlend {blendOption=Linear}
    sourceBone {boneName=thigh_l}
    targetBone {boneName=dyn_skirt_frdiag_A_01_l}
   }
   n67 = "dyn_skirt_frdiag_01_r.rotateY = thigh_r.rotateY * -0.1 - Bone Driven Controller"!(ComponentPose:Component Space Pose Link Structure=@n66, bAlphaBoolEnabled:bool=True) -> (Pose:Component Space Pose Link Structure) {
    is=title AnimGraphNode_BoneDrivenController
    alphaInputType=Bool
    bAffectTargetRotationY=true
    multiplier=-0.1
    sourceComponent=RotationY
    alphaBoolBlend {blendOption=Linear}
    sourceBone {boneName=thigh_r}
    targetBone {boneName=dyn_skirt_frdiag_01_r}
   }
   n68 = "dyn_skirt_frdiag_A_01_r.rotateY = thigh_r.rotateZ * -0.05 - Bone Driven Controller"!(ComponentPose:Component Space Pose Link Structure=@n67, bAlphaBoolEnabled:bool=True) -> (Pose:Component Space Pose Link Structure) {
    is=title AnimGraphNode_BoneDrivenController
    alphaInputType=Bool
    bAffectTargetRotationY=true
    multiplier=-0.05
    sourceComponent=RotationZ
    alphaBoolBlend {blendOption=Linear}
    sourceBone {boneName=thigh_r}
    targetBone {boneName=dyn_skirt_frdiag_A_01_r}
   }
   n69 = "Transform (Modify) Bone - Bone: dyn_skirt_fr_01_l"(ComponentPose=@n68, Alpha=Get L Leg Fwd Alpha() -> (L Leg Fwd Alpha:double), Translation="0,10,2", Rotation="-5,0,0", Scale="1,1,1") {
    rotationMode=BMM_Additive
    rotationSpace=BCS_BoneSpace
    translationMode=BMM_Additive
    translationSpace=BCS_BoneSpace
    alphaBoolBlend {blendOption=Linear}
    boneToModify {boneName=dyn_skirt_fr_01_l}
    rotation {pitch=-5}
    translation {y=10, z=2}
   }
   n70 = "Transform (Modify) Bone - Bone: dyn_skirt_fr_01_r"!(ComponentPose=@n69, Alpha:float=Get R Leg Fwd Alpha(), Translation="0,-10,-2", Rotation="-5,0,0", Scale="1,1,1") -> (Pose) {
    rotationMode=BMM_Additive
    rotationSpace=BCS_BoneSpace
    translationMode=BMM_Additive
    translationSpace=BCS_BoneSpace
    alphaBoolBlend {blendOption=Linear}
    boneToModify {boneName=dyn_skirt_fr_01_r}
    rotation {pitch=-5}
    translation {y=-10, z=-2}
   }
   n71 = "Constraint - dyn_skirt_fr_01 (PP)"(ComponentPose=@n70, Alpha=1, ConstraintWeights_0=0.5, ConstraintWeights_1=0.5) {
    alphaBoolBlend {blendOption=Linear}
    boneToModify {boneName=dyn_skirt_fr_01}
    constraintSetup [{targetBone {boneName=dyn_skirt_fr_01_l}}, {targetBone {boneName=dyn_skirt_fr_01_r}}]
    constraintWeights [1,1]
   }
   n72 = "Constraint - dyn_skirt_fr_02 (PP)"(ComponentPose=@n71, Alpha=1, ConstraintWeights_0=0.5, ConstraintWeights_1=0.5) {
    alphaBoolBlend {blendOption=Linear}
    boneToModify {boneName=dyn_skirt_fr_02}
    constraintSetup [{targetBone {boneName=dyn_skirt_fr_02_l}}, {targetBone {boneName=dyn_skirt_fr_02_r}}]
    constraintWeights [1,1]
   }
   n73 = "Constraint - dyn_skirt_fr_03 (PP)"(ComponentPose=@n72, Alpha=1, ConstraintWeights_0=0.5, ConstraintWeights_1=0.5) {
    alphaBoolBlend {blendOption=Linear}
    boneToModify {boneName=dyn_skirt_fr_03}
    constraintSetup [{targetBone {boneName=dyn_skirt_fr_03_l}}, {targetBone {boneName=dyn_skirt_fr_03_r}}]
    constraintWeights [1,1]
   }
   n74 = Anim Dynamics - Bone: dyn_skirt_fr_01(ComponentPose=@n73, Alpha=1, ExternalForce=@n52) {
    angularDampingOverride=0.98
    bGravityOverrideInSimSpace=true
    bOverrideAngularDamping=true
    bUseGravityOverride=true
    simulationSpace=RootRelative
    alphaBoolBlend {blendOption=Linear}
    boundBone {boneName=dyn_skirt_fr_01}
    componentAppliedLinearAccClamp=<3000 3000 3000>
    componentLinearAccScale=<0.1 0.1 0.1>
    componentLinearVelScale=<0.1 0.1 0.1>
    gravityOverride {x=1, z=-980}
    physicsBodyDefinitions [{
      sphereCollisionRadius=0
      boundBone {boneName=dyn_skirt_fr_01}
      boxExtents=<15 25 25>
      constraintSetup {angularLimitsMax=<0 0 5>, angularLimitsMin=<0 -10 -5>, angularTarget=<0 0 0>}
      localJointOffset=<3 0 0>
     }]
    relativeSpaceBone {boneName=root}
   }
   n75 = Anim Dynamics - Bone: dyn_skirt_fr_02(ComponentPose=@n74, Alpha=1, ExternalForce=@n52) {
    angularDampingOverride=0.98
    bGravityOverrideInSimSpace=true
    bOverrideAngularDamping=true
    bUseGravityOverride=true
    simulationSpace=RootRelative
    alphaBoolBlend {blendOption=Linear}
    boundBone {boneName=dyn_skirt_fr_02}
    componentAppliedLinearAccClamp=<3000 3000 3000>
    componentLinearAccScale=<0.1 0.1 0.1>
    componentLinearVelScale=<0.1 0.1 0.1>
    gravityOverride {x=1, z=-980}
    physicsBodyDefinitions [{
      sphereCollisionRadius=0
      boundBone {boneName=dyn_skirt_fr_02}
      boxExtents=<15 25 25>
      constraintSetup {angularLimitsMax=<10 5 5>, angularLimitsMin=<-10 -30 -5>, angularTarget=<0 0 0>}
      localJointOffset=<3 0 0>
     }]
    relativeSpaceBone {boneName=root}
   }
   n76 = Anim Dynamics - Bone: dyn_skirt_fr_03(ComponentPose=@n75, Alpha=1, ExternalForce=@n52) {
    angularDampingOverride=0.98
    bGravityOverrideInSimSpace=true
    bOverrideAngularDamping=true
    bUseGravityOverride=true
    simulationSpace=RootRelative
    alphaBoolBlend {blendOption=Linear}
    boundBone {boneName=dyn_skirt_fr_03}
    componentAppliedLinearAccClamp=<3000 3000 3000>
    componentLinearAccScale=<0.1 0.1 0.1>
    componentLinearVelScale=<0.1 0.1 0.1>
    gravityOverride {x=1, z=-980}
    physicsBodyDefinitions [{
      sphereCollisionRadius=0
      boundBone {boneName=dyn_skirt_fr_03}
      boxExtents=<15 25 25>
      constraintSetup {angularLimitsMax=<10 10 5>, angularLimitsMin=<-10 -30 -5>, angularTarget=<0 0 0>}
      localJointOffset=<3 0 0>
     }]
    relativeSpaceBone {boneName=root}
   }
   n77 = Get bIsCrouching()
   n78 = "Transform (Modify) Bone - Bone: dyn_skirt_out_01_r_MSTR"!(ComponentPose=@n76, bAlphaBoolEnabled:bool=@n77, Translation="0,0,0", Rotation="15,0,0", Scale="1,1,1") -> (Pose) {
    alphaInputType=Bool
    rotationMode=BMM_Additive
    rotationSpace=BCS_BoneSpace
    translationMode=BMM_Additive
    translationSpace=BCS_BoneSpace
    alphaBoolBlend {blendInTime=0.1, blendOption=Linear, blendOutTime=0.1}
    boneToModify {boneName=dyn_skirt_out_01_r_MSTR}
    rotation {pitch=15}
   }
   n79 = AND Boolean(A=@n77, B=NOT Boolean(A=Get bIsMoving2D()))
   n80 = "Transform (Modify) Bone - Bone: dyn_skirt_bk_01_r_MSTR"!(ComponentPose=@n78, bAlphaBoolEnabled:bool=@n79, Translation="0,0,0", Rotation="25,-10,0", Scale="1,1,1") -> (Pose) {
    alphaInputType=Bool
    rotationMode=BMM_Additive
    rotationSpace=BCS_BoneSpace
    translationMode=BMM_Additive
    translationSpace=BCS_BoneSpace
    alphaBoolBlend {blendInTime=0.025, blendOption=Linear, blendOutTime=0.25}
    boneToModify {boneName=dyn_skirt_bk_01_r_MSTR}
    rotation {pitch=25, yaw=-10}
   }
   n81 = "Transform (Modify) Bone - Bone: dyn_skirt_bk_A_01_r_MSTR"!(ComponentPose=@n80, bAlphaBoolEnabled:bool=@n79, Translation="0,0,0", Rotation="30,-10,0", Scale="1,1,1") -> (Pose) {
    alphaInputType=Bool
    rotationMode=BMM_Additive
    rotationSpace=BCS_BoneSpace
    translationMode=BMM_Additive
    translationSpace=BCS_BoneSpace
    alphaBoolBlend {blendInTime=0.025, blendOption=Linear, blendOutTime=0.25}
    boneToModify {boneName=dyn_skirt_bk_A_01_r_MSTR}
    rotation {pitch=30, yaw=-10}
   }
   n82 = Get bIsSprinting()
   n83 = "Transform (Modify) Bone - Bone: dyn_skirt_bk_01_r_MSTR"(ComponentPose=@n81, bAlphaBoolEnabled=@n82, Translation="0,0,0", Rotation="25,0,0", Scale="1,1,1") {
    alphaInputType=Bool
    rotationMode=BMM_Additive
    rotationSpace=BCS_BoneSpace
    translationMode=BMM_Additive
    translationSpace=BCS_BoneSpace
    alphaBoolBlend {blendInTime=0.1, blendOption=Linear, blendOutTime=0.1}
    boneToModify {boneName=dyn_skirt_bk_01_r_MSTR}
    rotation {pitch=25}
   }
   n84 = "Transform (Modify) Bone - Bone: dyn_skirt_bk_A_01_r_MSTR"(ComponentPose=@n83, bAlphaBoolEnabled=@n82, Translation="0,0,0", Rotation="25,0,0", Scale="1,1,1") {
    alphaInputType=Bool
    rotationMode=BMM_Additive
    rotationSpace=BCS_BoneSpace
    translationMode=BMM_Additive
    translationSpace=BCS_BoneSpace
    alphaBoolBlend {blendInTime=0.1, blendOption=Linear, blendOutTime=0.1}
    boneToModify {boneName=dyn_skirt_bk_A_01_r_MSTR}
    rotation {pitch=25}
   }
   n85 = "Transform (Modify) Bone - Bone: dyn_skirt_out_01_l_MSTR"!(ComponentPose=@n84, bAlphaBoolEnabled:bool=@n77, Translation="0,0,0", Rotation="15,0,0", Scale="1,1,1") -> (Pose) {
    alphaInputType=Bool
    rotationMode=BMM_Additive
    rotationSpace=BCS_BoneSpace
    translationMode=BMM_Additive
    translationSpace=BCS_BoneSpace
    alphaBoolBlend {blendInTime=0.1, blendOption=Linear, blendOutTime=0.1}
    boneToModify {boneName=dyn_skirt_out_01_l_MSTR}
    rotation {pitch=15}
   }
   n86 = "Transform (Modify) Bone - Bone: dyn_skirt_bk_MSTR"!(ComponentPose=@n85, bAlphaBoolEnabled:bool=@n79, Translation="0,0,0", Rotation="27,10,0", Scale="1,1,1") -> (Pose) {
    alphaInputType=Bool
    rotationMode=BMM_Additive
    rotationSpace=BCS_BoneSpace
    translationMode=BMM_Additive
    translationSpace=BCS_BoneSpace
    alphaBoolBlend {blendInTime=0.025, blendOption=Linear, blendOutTime=0.25}
    boneToModify {boneName=dyn_skirt_bk_MSTR}
    rotation {pitch=27, yaw=10}
   }
   n87 = "Transform (Modify) Bone - Bone: dyn_skirt_bk_01_l_MSTR"!(ComponentPose=@n86, bAlphaBoolEnabled:bool=@n79, Translation="0,0,0", Rotation="10,0,0", Scale="1,1,1") -> (Pose) {
    alphaInputType=Bool
    rotationMode=BMM_Additive
    rotationSpace=BCS_BoneSpace
    translationMode=BMM_Additive
    translationSpace=BCS_BoneSpace
    alphaBoolBlend {blendInTime=0.025, blendOption=Linear, blendOutTime=0.25}
    boneToModify {boneName=dyn_skirt_bk_01_l_MSTR}
    rotation {pitch=10}
   }
   n88 = "Transform (Modify) Bone - Bone: dyn_skirt_bk_A_01_l_MSTR"!(ComponentPose=@n87, bAlphaBoolEnabled:bool=@n79, Translation="0,0,0", Rotation="15,0,0", Scale="1,1,1") -> (Pose) {
    alphaInputType=Bool
    rotationMode=BMM_Additive
    rotationSpace=BCS_BoneSpace
    translationMode=BMM_Additive
    translationSpace=BCS_BoneSpace
    alphaBoolBlend {blendInTime=0.025, blendOption=Linear, blendOutTime=0.25}
    boneToModify {boneName=dyn_skirt_bk_A_01_l_MSTR}
    rotation {pitch=15}
   }
   n89 = "Transform (Modify) Bone - Bone: dyn_skirt_bk_01_l_MSTR"(ComponentPose=@n88, bAlphaBoolEnabled=@n82, Translation="0,0,0", Rotation="25,0,0", Scale="1,1,1") {
    alphaInputType=Bool
    rotationMode=BMM_Additive
    rotationSpace=BCS_BoneSpace
    translationMode=BMM_Additive
    translationSpace=BCS_BoneSpace
    alphaBoolBlend {blendInTime=0.1, blendOption=Linear, blendOutTime=0.1}
    boneToModify {boneName=dyn_skirt_bk_01_l_MSTR}
    rotation {pitch=25}
   }
   n90 = "Transform (Modify) Bone - Bone: dyn_skirt_bk_A_01_l_MSTR"(ComponentPose=@n89, bAlphaBoolEnabled=@n82, Translation="0,0,0", Rotation="25,0,0", Scale="1,1,1") {
    alphaInputType=Bool
    rotationMode=BMM_Additive
    rotationSpace=BCS_BoneSpace
    translationMode=BMM_Additive
    translationSpace=BCS_BoneSpace
    alphaBoolBlend {blendInTime=0.1, blendOption=Linear, blendOutTime=0.1}
    boneToModify {boneName=dyn_skirt_bk_A_01_l_MSTR}
    rotation {pitch=25}
   }
   n91 = "Transform (Modify) Bone - Bone: dyn_skirt_bk_MSTR"(ComponentPose=@n90, bAlphaBoolEnabled=@n82, Translation="0,0,0", Rotation="25,0,0", Scale="1,1,1") {
    alphaInputType=Bool
    rotationMode=BMM_Additive
    rotationSpace=BCS_BoneSpace
    translationMode=BMM_Additive
    translationSpace=BCS_BoneSpace
    alphaBoolBlend {blendInTime=0.1, blendOption=Linear, blendOutTime=0.1}
    boneToModify {boneName=dyn_skirt_bk_MSTR}
    rotation {pitch=25}
   }
   n92 = Get L Leg Fwd Alpha()
   n93 = "Transform (Modify) Bone - Bone: dyn_skirt_frdiag_A_01_l_MSTR"(ComponentPose=@n91, Alpha=@n92, Translation="-4,0,-3", Rotation="-7,0,0", Scale="1,1,1") {
    rotationMode=BMM_Additive
    rotationSpace=BCS_BoneSpace
    translationMode=BMM_Additive
    translationSpace=BCS_BoneSpace
    alphaBoolBlend {blendOption=Linear}
    boneToModify {boneName=dyn_skirt_frdiag_A_01_l_MSTR}
    rotation {pitch=-7}
    translation {x=-4, z=-3}
   }
   n94 = "Transform (Modify) Bone - Bone: dyn_skirt_frdiag_01_l_MSTR"(ComponentPose=@n93, Alpha=@n92, Translation="0,-1,2", Rotation="0,0,0", Scale="1,1,1") {
    rotationMode=BMM_Additive
    rotationSpace=BCS_BoneSpace
    translationMode=BMM_Additive
    translationSpace=BCS_BoneSpace
    alphaBoolBlend {blendOption=Linear}
    boneToModify {boneName=dyn_skirt_frdiag_01_l_MSTR}
    translation {y=-1, z=2}
   }
   n95 = Get L Leg Side Alpha() -> (L Leg Side Alpha:double)
   n96 = "Transform (Modify) Bone - Bone: dyn_skirt_frdiag_01_l_MSTR"(ComponentPose=@n94, Alpha=@n95, Translation="0,0,2", Rotation="50,0,0", Scale="1,1,1") {
    rotationMode=BMM_Additive
    rotationSpace=BCS_BoneSpace
    translationMode=BMM_Additive
    translationSpace=BCS_BoneSpace
    alphaBoolBlend {blendOption=Linear}
    boneToModify {boneName=dyn_skirt_frdiag_01_l_MSTR}
    rotation {pitch=50}
    translation {z=2}
   }
   n97 = "Transform (Modify) Bone - Bone: dyn_skirt_out_01_l_MSTR"!(ComponentPose=@n96, Alpha:float=@n95, Translation="0,0,1", Rotation="40,10,0", Scale="1,1,1") -> (Pose) {
    rotationMode=BMM_Additive
    rotationSpace=BCS_BoneSpace
    translationMode=BMM_Additive
    translationSpace=BCS_BoneSpace
    alphaBoolBlend {blendOption=Linear}
    boneToModify {boneName=dyn_skirt_out_01_l_MSTR}
    rotation {pitch=40, yaw=10}
    translation {z=1}
   }
   n98 = Get R Leg Fwd Alpha()
   n99 = "Transform (Modify) Bone - Bone: dyn_skirt_frdiag_01_r_MSTR"(ComponentPose=@n97, Alpha=@n98, Translation="0,6,7", Rotation="15,0,0", Scale="1,1,1") {
    rotationMode=BMM_Additive
    rotationSpace=BCS_BoneSpace
    translationMode=BMM_Additive
    translationSpace=BCS_BoneSpace
    alphaBoolBlend {blendOption=Linear}
    boneToModify {boneName=dyn_skirt_frdiag_01_r_MSTR}
    rotation {pitch=15}
    translation {y=6, z=7}
   }
   n100 = "Transform (Modify) Bone - Bone: dyn_skirt_frdiag_01_r_MSTR"(ComponentPose=@n99, Alpha=@n98, Translation="0,-2,2", Rotation="15,0,0", Scale="1,1,1") {
    rotationMode=BMM_Additive
    rotationSpace=BCS_BoneSpace
    translationMode=BMM_Additive
    translationSpace=BCS_BoneSpace
    alphaBoolBlend {blendOption=Linear}
    boneToModify {boneName=dyn_skirt_frdiag_01_r_MSTR}
    rotation {pitch=15}
    translation {y=-2, z=2}
   }
   n101 = "Transform (Modify) Bone - Bone: dyn_skirt_out_01_r_MSTR"!(ComponentPose=@n100, Alpha:float=Get R Leg Side Alpha(), Translation="0,0,1", Rotation="25,0,0", Scale="1,1,1") -> (Pose) {
    rotationMode=BMM_Additive
    rotationSpace=BCS_BoneSpace
    translationMode=BMM_Additive
    translationSpace=BCS_BoneSpace
    alphaBoolBlend {blendOption=Linear}
    boneToModify {boneName=dyn_skirt_out_01_r_MSTR}
    rotation {pitch=25}
    translation {z=1}
   }
   n102 = Get SkydiveAcc() -> (SkydiveAcc:Vector)
   n103 = RigidBody!(
    ComponentPose=@n101
    Alpha
    ExternalForce=Get Skydiving Vector Force()
    ComponentLinearAccScale:Vector=@n102
    ComponentLinearVelScale:Vector=@n102
   ) -> (Pose) {
    bClampLinearTranslationLimitToRefPose=true
    bOverrideWorldGravity=true
    overridePhysicsAsset=&/F_MED_DeLasMuertos_Physics_perf2
    simulationSpace=BaseBoneSpace
    alphaBoolBlend {blendOption=Linear}
    baseBoneRef {boneName=pelvis}
    componentAppliedLinearAccClamp=<3000 3000 3000>
    componentLinearAccScale=<0.3 0.3 0.6>
    componentLinearVelScale=<0.3 0.3 0.6>
    overrideWorldGravity {z=-980}
   }
   n104 = "Constraint - dyn_hip_rose (PPPPPP)"!(
    ComponentPose=@n103
    Alpha=1
    ConstraintWeights_0=0.642
    ConstraintWeights_1:float=0.1163
    ConstraintWeights_2:float=0.0057
    ConstraintWeights_3:float=0.0001
    ConstraintWeights_4:float=0.1706
    ConstraintWeights_5:float=0.0653
   ) -> (Pose) {
    alphaBoolBlend {blendOption=Linear}
    boneToModify {boneName=dyn_hip_rose}
    constraintSetup [
     {targetBone {boneName=pelvis}}
     {targetBone {boneName=dyn_skirt_fr_01_r}}
     {targetBone {boneName=dyn_skirt_fr_01}}
     {targetBone {boneName=dyn_skirt_fr_02_r}}
     {targetBone {boneName=dyn_skirt_frdiag_A_01_r}}
     {targetBone {boneName=dyn_skirt_frdiag_01_r}}
    ]
    constraintWeights [1,1,1,1,1,1]
   }
   n105 = "Transform (Modify) Bone - Bone: dyn_hip_rose"(ComponentPose=@n104, Alpha=Get R Leg Fwd Alpha(), Translation="0,1,1", Rotation="0,0,-35", Scale="1,1,1") {
    rotationMode=BMM_Additive
    rotationSpace=BCS_BoneSpace
    translationMode=BMM_Additive
    translationSpace=BCS_BoneSpace
    alphaBoolBlend {blendOption=Linear}
    boneToModify {boneName=dyn_hip_rose}
    rotation {roll=-35}
    translation {y=1, z=1}
   }
   n106 = "Transform (Modify) Bone - Bone: dyn_hip_rose"(ComponentPose=@n105, Alpha=Get R Leg Side Alpha(), Translation="0,0.5,0", Rotation="0,10,-10", Scale="1,1,1") {
    rotationMode=BMM_Additive
    rotationSpace=BCS_BoneSpace
    translationMode=BMM_Additive
    translationSpace=BCS_BoneSpace
    alphaBoolBlend {blendOption=Linear}
    boneToModify {boneName=dyn_hip_rose}
    rotation {roll=-10, yaw=10}
    translation {y=0.5}
   }
   n107 = Anim Dynamics - Bone: dyn_hip_rose(ComponentPose=@n106, Alpha=0.3, ExternalForce="0,0,0") {
    angularDampingOverride=0.99
    angularSpringConstant=1500
    bAngularSpring=true
    bGravityOverrideInSimSpace=true
    bOverrideAngularDamping=true
    gravityScale=0.35
    numSolverIterationsPostUpdate=2
    numSolverIterationsPreUpdate=8
    simulationSpace=RootRelative
    alphaBoolBlend {blendOption=Linear}
    boundBone {boneName=dyn_hip_rose}
    componentAppliedLinearAccClamp=<3000 3000 3000>
    componentLinearAccScale=<0.2 0.2 0.2>
    componentLinearVelScale=<0.2 0.2 0.2>
    physicsBodyDefinitions [{
      sphereCollisionRadius=0
      boundBone {boneName=dyn_hip_rose}
      boxExtents=<20 20 10>
      constraintSetup {angularTargetAxis=AxisZ, twistAxis=AxisZ, angularLimitsMax=<3 3 3>, angularLimitsMin=<-3 -3 -3>, angularTarget=<0 0 0>}
      localJointOffset=<0 0 -2>
     }]
   }
   n108 = "Constraint - dyn_hip_bow_l (PPPPP)"!(
    ComponentPose=@n107
    Alpha=1
    ConstraintWeights_0=0.0001
    ConstraintWeights_1:float=0.0001
    ConstraintWeights_2:float=0
    ConstraintWeights_3:float=0.0001
    ConstraintWeights_4:float=0.9997
   ) -> (Pose) {
    alphaBoolBlend {blendOption=Linear}
    boneToModify {boneName=dyn_hip_bow_l}
    constraintSetup [
     {targetBone {boneName=spine_01}}
     {targetBone {boneName=spine_02}}
     {targetBone {boneName=spine_04}}
     {targetBone {boneName=thigh_twist_01_l}}
     {targetBone {boneName=pelvis}}
    ]
    constraintWeights [1,1,1,1,1]
   }
   n109 = "Transform (Modify) Bone - Bone: dyn_hip_bow_l"(ComponentPose=@n108, bAlphaBoolEnabled=Get bIsCrouching(), Translation="-1,0,-4", Rotation="-15,-5,-35", Scale="1,1,1") {
    alphaInputType=Bool
    rotationMode=BMM_Additive
    rotationSpace=BCS_BoneSpace
    translationMode=BMM_Additive
    translationSpace=BCS_BoneSpace
    alphaBoolBlend {blendInTime=0.1, blendOption=Linear, blendOutTime=0.1}
    boneToModify {boneName=dyn_hip_bow_l}
    rotation=<rot -15 -5 -35>
    translation {x=-1, z=-4}
   }
   n110 = Get bIsCrouching()
   n111 = "Transform (Modify) Bone - Bone: dyn_hip_bow_l"!(
    ComponentPose=@n109
    Alpha:float=SelectFloat(A=Get L Leg Side Alpha(), B=0, bPickA=NOT Boolean(A=@n110))
    Translation="0,0,0"
    Rotation="0,-45,0"
    Scale="1,1,1"
   ) -> (Pose) {
    rotationMode=BMM_Additive
    rotationSpace=BCS_BoneSpace
    translationMode=BMM_Additive
    translationSpace=BCS_BoneSpace
    alphaBoolBlend {blendInTime=0.1, blendOption=Linear, blendOutTime=0.1}
    boneToModify {boneName=dyn_hip_bow_l}
    rotation {yaw=-45}
   }
   n112 = SelectFloat(A=Get L Leg Side Alpha(), B=0, bPickA=Get bIsCrouching())
   n113 = "Transform (Modify) Bone - Bone: dyn_hip_bow_l"(ComponentPose=@n111, Alpha=@n112, Translation="0,0,0", Rotation="0,-10,0", Scale="1,1,1") {
    rotationMode=BMM_Additive
    rotationSpace=BCS_BoneSpace
    translationMode=BMM_Additive
    translationSpace=BCS_BoneSpace
    alphaBoolBlend {blendInTime=0.1, blendOption=Linear, blendOutTime=0.1}
    boneToModify {boneName=dyn_hip_bow_l}
    rotation {yaw=-10}
   }
   n114 = "Transform (Modify) Bone - Bone: dyn_hip_bow_string01_r"(ComponentPose=@n113, Alpha=@n112, Translation="0,-5,-2", Rotation="0,-15,0", Scale="1,1,1") {
    rotationMode=BMM_Additive
    rotationSpace=BCS_BoneSpace
    translationMode=BMM_Additive
    translationSpace=BCS_BoneSpace
    alphaBoolBlend {blendInTime=0.1, blendOption=Linear, blendOutTime=0.1}
    boneToModify {boneName=dyn_hip_bow_string01_r}
    rotation {yaw=-15}
    translation {y=-5, z=-2}
   }
   n115 = "Transform (Modify) Bone - Bone: dyn_hip_bow_string01_l"(ComponentPose=@n114, Alpha=@n112, Translation="0,-2,-1", Rotation="0,-15,0", Scale="1,1,1") {
    rotationMode=BMM_Additive
    rotationSpace=BCS_BoneSpace
    translationMode=BMM_Additive
    translationSpace=BCS_BoneSpace
    alphaBoolBlend {blendInTime=0.1, blendOption=Linear, blendOutTime=0.1}
    boneToModify {boneName=dyn_hip_bow_string01_l}
    rotation {yaw=-15}
    translation {y=-2, z=-1}
   }
   n116 = Anim Dynamics - Bone: dyn_hip_bow_string01_r(ComponentPose=@n115, Alpha=1, ExternalForce="0,0,0") {
    angularDampingOverride=0.99
    bGravityOverrideInSimSpace=true
    bOverrideAngularDamping=true
    gravityScale=0.35
    numSolverIterationsPostUpdate=2
    numSolverIterationsPreUpdate=8
    simulationSpace=RootRelative
    alphaBoolBlend {blendOption=Linear}
    boundBone {boneName=dyn_hip_bow_string01_r}
    componentAppliedLinearAccClamp=<3000 3000 3000>
    componentLinearAccScale=<0.2 0.2 0.35>
    componentLinearVelScale=<0.2 0.2 0.35>
    physicsBodyDefinitions [{
      sphereCollisionRadius=0
      boundBone {boneName=dyn_hip_bow_string01_r}
      boxExtents=<25 7 25>
      constraintSetup {angularTargetAxis=AxisZ, twistAxis=AxisZ, angularLimitsMax=<2 4 -2>, angularLimitsMin=<-2 -20 -15>, angularTarget=<0 0 0>}
      localJointOffset=<3 0 0>
     }]
    relativeSpaceBone {boneName=dyn_hip_bow_l}
   }
   n117 = Copy Bone - Source Bone: dyn_hip_bow_string01_r - Target Bone: dyn_hip_bow_string02_r!(
    ComponentPose:Component Space Pose Link Structure=@n116
    Alpha:float=0.7
    bCopyTranslation:bool=False
    bCopyRotation:bool=true
    bCopyScale:bool=False
   ) -> (Pose:Component Space Pose Link Structure) {
    is=title AnimGraphNode_CopyBone
    alphaBoolBlend {blendOption=Linear}
    sourceBone {boneName=dyn_hip_bow_string01_r}
    targetBone {boneName=dyn_hip_bow_string02_r}
   }
   n118 = Anim Dynamics - Bone: dyn_hip_bow_string01_l(ComponentPose=@n117, Alpha=1, ExternalForce="0,0,0") {
    angularDampingOverride=0.987
    bGravityOverrideInSimSpace=true
    bOverrideAngularDamping=true
    gravityScale=0.35
    numSolverIterationsPostUpdate=2
    numSolverIterationsPreUpdate=8
    simulationSpace=RootRelative
    alphaBoolBlend {blendOption=Linear}
    boundBone {boneName=dyn_hip_bow_string01_l}
    componentAppliedLinearAccClamp=<3000 3000 3000>
    componentLinearAccScale=<0.2 0.2 0.35>
    componentLinearVelScale=<0.2 0.2 0.35>
    physicsBodyDefinitions [{
      sphereCollisionRadius=0
      boundBone {boneName=dyn_hip_bow_string01_l}
      boxExtents=<25 7 25>
      constraintSetup {angularTargetAxis=AxisZ, twistAxis=AxisZ, angularLimitsMax=<2 4 -2>, angularLimitsMin=<-2 -20 -15>, angularTarget=<0 0 0>}
      localJointOffset=<3 0 0>
     }]
    relativeSpaceBone {boneName=dyn_hip_bow_l}
   }
   n119 = Copy Bone - Source Bone: dyn_hip_bow_string01_l - Target Bone: dyn_hip_bow_string02_l!(
    ComponentPose:Component Space Pose Link Structure=@n118
    Alpha:float=0.8
    bCopyTranslation:bool=False
    bCopyRotation:bool=true
    bCopyScale:bool=False
   ) -> (Pose:Component Space Pose Link Structure) {
    is=title AnimGraphNode_CopyBone
    alphaBoolBlend {blendOption=Linear}
    sourceBone {boneName=dyn_hip_bow_string01_l}
    targetBone {boneName=dyn_hip_bow_string02_l}
   }
   n120 = "Transform (Modify) Bone - Bone: dyn_hip_bow_string01_r"!(ComponentPose=@n119, bAlphaBoolEnabled:bool=True, Translation="0,-1,0", Rotation="0,0,0", Scale="1,1,1") -> (Pose) {
    alphaInputType=Bool
    rotationMode=BMM_Additive
    rotationSpace=BCS_BoneSpace
    translationMode=BMM_Additive
    translationSpace=BCS_BoneSpace
    alphaBoolBlend {blendInTime=0.1, blendOption=Linear, blendOutTime=0.1}
    boneToModify {boneName=dyn_hip_bow_string01_r}
    translation {y=-1}
   }
   n121 = "Transform (Modify) Bone - Bone: dyn_hip_bow_string01_l"!(ComponentPose=@n120, bAlphaBoolEnabled:bool=True, Translation="0,-1,0", Rotation="0,0,0", Scale="1,1,1") -> (Pose) {
    alphaInputType=Bool
    rotationMode=BMM_Additive
    rotationSpace=BCS_BoneSpace
    translationMode=BMM_Additive
    translationSpace=BCS_BoneSpace
    alphaBoolBlend {blendInTime=0.1, blendOption=Linear, blendOutTime=0.1}
    boneToModify {boneName=dyn_hip_bow_string01_l}
    translation {y=-1}
   }
   Output Pose(Result=Component To Local(ComponentPose=@n121)) {layerGroup=None}
   "To Float (Boolean)"(InBool=@n20)
   "To Float (Boolean)"(InBool=@n20)
   Reroute Node(InputPin:Wildcard) -> (OutputPin:Wildcard)
   n122 = "Transform (Modify) Bone - Bone: dyn_skirt_fr_01_l"(Alpha=Get L Leg Fwd Alpha(), Translation="0,5,0", Rotation="-5,0,0", Scale="1,1,1") {
    rotationMode=BMM_Additive
    rotationSpace=BCS_BoneSpace
    scaleMode=BMM_Additive
    scaleSpace=BCS_BoneSpace
    translationMode=BMM_Replace
    translationSpace=BCS_BoneSpace
    alphaBoolBlend {blendOption=Linear}
    boneToModify {boneName=dyn_skirt_fr_01_l}
    rotation {pitch=-5}
    translation {y=5}
   }
   n123 = "Transform (Modify) Bone - Bone: dyn_skirt_fr_01_r"(ComponentPose=@n122, Alpha=Get thigh_shape_fn_r() -> (thigh_shape_fn_r:double), Translation="0,-10,0", Rotation="-5,0,0", Scale="1,1,1") {
    rotationMode=BMM_Additive
    rotationSpace=BCS_BoneSpace
    scaleMode=BMM_Additive
    scaleSpace=BCS_BoneSpace
    translationMode=BMM_Additive
    translationSpace=BCS_BoneSpace
    alphaBoolBlend {blendOption=Linear}
    boneToModify {boneName=dyn_skirt_fr_01_r}
    rotation {pitch=-5}
    translation {y=-10}
   }
   Reroute Node(InputPin:Component Space Pose Link Structure=@n123) -> (OutputPin:Component Space Pose Link Structure)
   Get thigh_shape_fn_l() -> (thigh_shape_fn_l:double)
   Get is In Air() -> (is In Air:bool)
  ]}
 EventGraph {nodes [
   Event BlueprintUpdateAnimation()
   > n1 = Sequence!(execute) -> (then_0, then_1, then_2:Exec, then_3:Exec, then_4:Exec)
   Set isCrouchOrSprint(execute=@n1.then_1, isCrouchOrSprint:bool=OR Boolean(A=Get bIsCrouching(), B=Get bIsMoving2D())) -> (Output_Get:bool)
   > Set is In Air(is In Air:bool=OR Boolean(A=Get bIsJumping(), B=Get bIsFalling())) -> (Output_Get:bool)
   Reroute Node(InputPin:Exec=@n1.then_3) -> (OutputPin:Exec)
   n2 = GetOwnerSkeletalMeshComponent()
   n3 = BreakRotator(InRot=BreakTransform(InTransform=GetDeltaTransformFromRefPose(self=@n2, BoneName=pelvis, BaseName=thigh_r)).Rotation)
   Set R Leg Fwd Alpha(execute=@n1.then_2, R Leg Fwd Alpha:double=MapRangeClamped(Value=@n3.Yaw, InRangeA=0, InRangeB=90)) -> (Output_Get:double)
   > Set R Leg Side Alpha(R Leg Side Alpha:double=MapRangeClamped(Value=@n3.Pitch, InRangeA=5, InRangeB=20)) -> (Output_Get:double)
   n4 = BreakTransform(InTransform=GetDeltaTransformFromRefPose(self=GetOwnerSkeletalMeshComponent(), BoneName=thigh_l, BaseName=pelvis))
   > Set L Leg Side Alpha(L Leg Side Alpha:double=MapRangeClamped(Value=BreakRotator(InRot=@n4.Rotation).Pitch, InRangeA=5, InRangeB=25)) -> (Output_Get:double)
   n5 = BreakTransform(InTransform=GetDeltaTransformFromRefPose(self=GetOwnerSkeletalMeshComponent(), BoneName=pelvis, BaseName=thigh_l))
   > Set L Leg Fwd Alpha(L Leg Fwd Alpha:double=MapRangeClamped(Value=BreakRotator(InRot=@n5.Rotation).Yaw, InRangeA=0, InRangeB=90)) -> (Output_Get:double)
   Set Skydiving Vector Force(
    execute=@n1.then_4
    Skydiving Vector Force:Vector=Break Fort Anim Input Skydiving External Force(FortAnimInput_SkydivingExternalForce=Get SkydivingExternalForceInput()) -> (
     bUseSkydivingVectorForce=~
     SkydivingVectorForce=~
     HeadToPelvisDirection=~
     FloatingMultiplier=~
     FloatingAdditive=~
     DivingMultiplier=~
     DivingAdditive=~
     ParachutingMultiplier=~
     ParachutingAdditive=~
    ).SkydivingVectorForce
   ) -> (Output_Get:Vector)
   > n6 = Branch(Condition=Get bIsSkyDiving())
   > Set SkydiveAcc(SkydiveAcc:Vector="0.1,0.1,0.2") -> (Output_Get:Vector)
   Set SkydiveAcc(execute=@n6.else, SkydiveAcc="0.3,0.3,0.6")
   Event OnParachuteOpened()
   > n7 = Delay(Duration=1.5)
   ResetDynamics(execute=[@n7.then, @n8.then, @n9.then], InTeleportType=TeleportPhysics)
   > ResetDynamics(InTeleportType=ResetPhysics)
   n8 = Event OnBeginSkydiving()
   n9 = Event OnEndSkydiving()
   Event BlueprintInitializeAnimation()
   > Set Skel_Mesh(Skel_Mesh=GetOwnerSkeletalMeshComponent())
   n10 = Get Skel_Mesh()
   Set thigh_shape_fn_l(thigh_shape_fn_l:double=GetAlphaFromReferenceTransformDelta(SkeletalMeshComponent=@n10, BoneName=thigh_l, BaseName=pelvis, InputRange=(X=-10,Y=-90), TwistAxis="1,0,0")) -> (Output_Get:double)
   > Set thigh_shape_fn_r(thigh_shape_fn_r:double=GetAlphaFromReferenceTransformDelta(SkeletalMeshComponent=@n10, BoneName=thigh_r, BaseName=pelvis, InputRange=(X=-10,Y=-90), TwistAxis="1,0,0")) -> (Output_Get:double)
   TryGetPawnOwner()
   Get bIsSprinting()
   Get bIsPlayingEmote()
   Set SkydiveAlpha(SkydiveAlpha:double=0) -> (Output_Get:double)
   Set SkydiveAlpha(SkydiveAlpha=0.5)
  ]}
]
variables [
 alphaVal {type=number, default=0.7}
 is In Air {type=boolean}
 isCrouchOrSprint {type=boolean}
 L Leg Fwd Alpha {type=number}
 L Leg Side Alpha {type=number}
 R Leg Fwd Alpha {type=number}
 R Leg Side Alpha {type=number}
 Skel_Mesh {type=SkeletalMeshComponent}
 SkydiveAcc {type=Vector}
 SkydiveAlpha {type=number}
 Skydiving Vector Force {type=Vector}
 thigh_shape_fn_l {type=number}
 thigh_shape_fn_r {type=number}
]
