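# Physics asset of cape colliders for the F_MED Muertos female body, apparently the collision volumes that its simulated cape collides with.
# Capsules cover the head, clavicles, spine, pelvis and feet, with crossed capsules on the middle spine bones.
# Large capsules on the dynamic skirt bones and left hip bow form collision volumes that follow those bones.
# Arm, thigh, calf and neck bodies are empty, and no constraints exist.
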PhysicsAsset /BRCosmetics/Characters/Player/Female/Medium/Bodies/F_MED_DeLosMuertos/Meshes/F_MED_Muertos_CapeColliders
bodies {
 calf_l []
 calf_r []
 clavicle_l [{shapeType=Capsule, length=3.9464, radius=10.1353, center=<10.8455 -0.2595 -1.2307>, rotation=<rot -77.5566 177.3915 7.7314>}]
 clavicle_r [{shapeType=Capsule, length=6.1964, radius=9.7957, center=<-9.3887 0.3607 2.0841>, rotation=<rot -77.5565 177.3915 -172.2685>}]
 dyn_hip_bow_l [{
   shapeType=Capsule
   length=8.5
   radius=9.75
   shapeName=dyn_hip_bow_l_capsule
   center=<8.4982 1.93 0.8471>
   rotation=<rot 79.9999 -179.9997 -179.9997>
  }]
 dyn_skirt_bk_05 [{shapeType=Capsule, length=41.25, radius=15, shapeName=dyn_skirt_bk_05_capsule, center=<-23.6435 0 -2.0871>, rotation=<rot 90 0 0>}]
 dyn_skirt_bk_05_l [{shapeType=Capsule, length=41.25, radius=15, shapeName=dyn_skirt_bk_05_l_capsule, center=<-23.6435 5.1673 -4.169>, rotation=<rot 80 0.0001 0>}]
 dyn_skirt_bk_05_r [{shapeType=Capsule, length=37, radius=15, shapeName=dyn_skirt_bk_05_r_capsule, center=<-23.7263 -6.2284 -4.6384>, rotation=<rot 80 -0.0002 0>}]
 dyn_skirt_bk_A_05_l [{
   shapeType=Capsule
   length=41.25
   radius=15
   shapeName=dyn_skirt_bk_A_05_l_capsule
   center=<-23.6435 1.7328 -4.0594>
   rotation=<rot 90 -166.7373 -166.7374>
  }]
 dyn_skirt_bk_A_05_r [{
   shapeType=Capsule
   length=41.25
   radius=15
   shapeName=dyn_skirt_bk_A_05_r_capsule
   center=<-23.6435 -2.8751 -2.9477>
   rotation=<rot 90 180 -180>
  }]
 dyn_skirt_frdiag_05_l [{
   shapeType=Capsule
   length=34.25
   radius=15
   shapeName=dyn_skirt_frdiag_05_l_capsule
   center=<-10.9188 0 -3.7946>
   rotation=<rot 90 180 179.9998>
  }]
 dyn_skirt_frdiag_05_r [{
   shapeType=Capsule
   length=34.25
   radius=15
   shapeName=dyn_skirt_frdiag_05_r_capsule
   center=<-9.0029 0 -3.7946>
   rotation=<rot 90 174.9488 174.949>
  }]
 foot_l [{shapeType=Capsule, length=12.5982, radius=6.5127, center=<-2.2072 6.0534 -0.3367>, rotation=<rot 0.0388 1.4701 -89.3982>}]
 foot_r [{shapeType=Capsule, length=12.5982, radius=6.5127, center=<2.2072 -6.0534 0.3367>, rotation=<rot 0.0388 1.4701 90.6018>}]
 head [{shapeType=Capsule, length=0, radius=18, center=<9.0697 0.0746 0>, rotation=<rot 0 -90 0>}]
 lowerarm_l []
 lowerarm_r []
 neck_02 []
 pelvis [{shapeType=Capsule, length=11.9028, radius=16.6595, center=<-3.9974 -5.4614 -0.1981>}]
 spine_01 [{shapeType=Capsule, length=12.5171, radius=10.2354, center=<-2.8183 -1.0873 0>}]
 spine_02 [
  {shapeType=Capsule, length=6.5484, radius=8.4854, center=<1.179 2.2974 0>}
  {shapeType=Capsule, length=20.6875, radius=9.3125, center=<-1.8317 2.606 0>, rotation=<rot 90 0 0>}
 ]
 spine_03 [
  {shapeType=Capsule, length=8.0484, radius=10.4854, center=<-0.0747 -3.6532 0>}
  {shapeType=Capsule, length=20.7188, radius=9.2812, center=<-1.2326 2.6059 0>, rotation=<rot 90 0 0>}
 ]
 spine_04 [
  {shapeType=Capsule, length=6.8296, radius=12.6416, center=<5.5182 -0.7216 0>}
  {shapeType=Capsule, length=10.0312, radius=9.2812, center=<-1.2326 1.308 0>, rotation=<rot 90 0 0>}
 ]
 spine_05 [{shapeType=Capsule, length=10.9859, radius=10.0791, center=<5.6377 0.3227 0>}]
 thigh_l []
 thigh_r []
 upperarm_l []
 upperarm_r []
}
constraints []
